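# Character material for HydroBottle that blends in an alternate "Feral" head texture set as a facial expression on the 2023 character parent shading.
# The saturated Normal_Curve scalar drives base color and normal blends between the PreFX output and the expression textures, with roughness also blended from SRM green.
# Expression base color first runs through the shared skin function using packed mask inputs.
# The blended attributes then feed the PostFX function.
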Material /BRCosmetics/Characters/Player/Male/Medium/Bodies/M_MED_HydroBottle/Materials/M_HydroBottle_FacialExpression
graph {
 properties {
  bUseMaterialAttributes=true
  bUsedWithMeshDeformer=true
  bUsedWithMorphTargets=true
  bUsedWithSkeletalMesh=true
  pixelDepthOffsetMode=PDOM_Legacy
  shadingModel=MSM_FromMaterialExpression
  usedShadingModels=Default Lit | Subsurface | Clear Coat | Subsurface Profile | Cloth
 }
 nodes [
  n1 = MaterialFunctionCall() {materialFunction=&MF_CharacterParent_2023_PreFX}
  n2 = GetMaterialAttributes(None=@n1.Result) {attributeGetTypes [
    #1
    #2
    #3
    #4
    #5
   ]}
  n3 = TextureSampleParameter2D() {
   group=Facial Expression
   parameterName=Basecolor_FacialExpression
   texture=/BRCosmetics/Characters/Player/Male/Medium/Bodies/M_MED_HydroBottle/Textures/T_Hydrobottle_Feral_Head_D
  }
  n4 = MaterialFunctionCall!(
   Skin Mask=NamedRerouteUsage()
   GMap=NamedRerouteUsage()
   Material=SetMaterialAttributes!(MaterialAttributes, Base Color=@n3.RGB) {attributeSetTypes [#1]}
  ) {materialFunction=/Game/Global/Materials/Character_Parent_2023/Functions/MF_CharacterParent_2023_Skin}
  n5 = Saturate(None=ScalarParameter() {group=Facial Expression, parameterName=Normal_Curve})
  n6 = TextureSampleParameter2D() {
   group=Facial Expression
   parameterName=SRM_FacialExpression
   samplerType=SAMPLERTYPE_LinearColor
   texture=/BRCosmetics/Characters/Player/Male/Medium/Bodies/M_MED_HydroBottle/Textures/T_Hydrobottle_Feral_Head_S
  }
  n7 = TextureSampleParameter2D() {
   group=Facial Expression
   parameterName=Normal_FacialExpression
   samplerType=SAMPLERTYPE_Normal
   texture=/BRCosmetics/Characters/Player/Male/Medium/Bodies/M_MED_HydroBottle/Textures/T_Hydrobottle_Feral_Head_N
  }
  n8 = SetMaterialAttributes!(
   MaterialAttributes=@n2.MaterialAttributes
   Base Color=LinearInterpolate(A=@n2.Base Color, B=GetMaterialAttributes(None=@n4.Result) {attributeGetTypes [#1]}.Base Color, Alpha=@n5)
   Metallic=@n2.MaterialAttributes
   Specular=@n2.MaterialAttributes
   Roughness=LinearInterpolate(A=@n2.Roughness, B=@n6.G, Alpha=NamedRerouteUsage())
   Normal=LinearInterpolate(A=@n2.Normal, B=@n7.RGB, Alpha=@n5)
  ) {attributeSetTypes [
    #1
    #2
    #3
    #4
    #5
   ]}
  MaterialFunctionCall(Material=@n8, M=@n1.M, Pre Skinned Local Position=@n1.M, Distance Blend=@n1.M, Fresnel=@n1.M) {materialFunction=&MF_CharacterParent_2023_PostFX}
  n9 = TextureSampleParameter2D() {
   group=Facial Expression
   parameterName=M_FacialExpression
   samplerType=SAMPLERTYPE_LinearColor
   texture=/BRCosmetics/Characters/Player/Male/Medium/Bodies/M_MED_HydroBottle/Textures/T_Hydrobottle_Feral_Head_M
  }
  n10 = MaterialFunctionCall(Float3=@n9.RGB) {desc="R = AO\r\nG = GMap\r\nB = Skin", materialFunction=&BreakOutFloat3Components}
  NamedRerouteDeclaration(None=@n10.R) {name=Ambient Occlusion, nodeColor=<rgba 0.875 0 1 1>}
  NamedRerouteDeclaration(None=@n10.G) {name=GMap, nodeColor=<rgba 0.6641 0 1 1>}
  NamedRerouteDeclaration(None=@n10.B) {name=Skin, nodeColor=<rgba 0.6641 0 1 1>}
  NamedRerouteDeclaration(None=@n5) {name=Alpha, nodeColor=<rgba 0.9922 1 0 1>}
 ]
}
