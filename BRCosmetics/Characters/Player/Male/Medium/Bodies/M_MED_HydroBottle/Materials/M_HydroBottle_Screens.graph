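# Character material for the HydroBottle body that draws animated sci-fi screens with typing text and heartline graphs.
# Five text areas, split by mask thresholds, scroll line by line on floored, sine-modulated time over scanline backgrounds, using bounding-box masks.
# Graph areas pan heartline textures with sphere-mask glows layered on top.
# Screen color replaces emissive through tech-art mask channels between the shared character PreFX and PostFX functions.
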Material /BRCosmetics/Characters/Player/Male/Medium/Bodies/M_MED_HydroBottle/Materials/M_HydroBottle_Screens
graph {
 properties {
  bUseMaterialAttributes=true
  bUsedWithMeshDeformer=true
  bUsedWithMorphTargets=true
  bUsedWithSkeletalMesh=true
  pixelDepthOffsetMode=PDOM_Legacy
  shadingModel=MSM_FromMaterialExpression
  usedShadingModels=Default Lit | Subsurface | Clear Coat | Subsurface Profile | Cloth
 }
 nodes [
  n1 = VectorParameter() {
   group=Scifi_Screen
   parameterName=TextTexture_Center01
   channelNames {a=ScaleY, b=ScaleX, g=CenterY, r=CenterX}
   defaultValue=<rgba 0.5 0.5 1 1>
  }
  MaterialExpressionAppendVector_19 = AppendVector(A=@n1.CenterX, B=@n1.CenterX)
  MaterialExpressionAppendVector_20 = AppendVector(A=@n1.ScaleX, B=@n1.ScaleX)
  MaterialExpressionStaticSwitchParameter_0 = StaticSwitchParameter(True=TextureCoordinate() {coordinateIndex=1}, False=TextureCoordinate()) {defaultValue=true, group=Scifi_Screen_Graph, parameterName=UseTexcoord01}
  n2 = Custom!(
   UVPoint=@MaterialExpressionAppendVector_19
   UVScale=@MaterialExpressionAppendVector_20
   OriginalUV=@MaterialExpressionStaticSwitchParameter_0
  ) {
   code="float2 DistUV =OriginalUV-UVPoint;\r\nDistUV = DistUV/UVScale + float2(0.5,0.5);\r\n\r\nreturn DistUV;"
   description=CopyToPointUV
   outputType=CMOT_Float2
   inputs [
    {
     inputName=UVPoint
     input {expression=&self:MaterialExpressionAppendVector_19}
    }
    {
     inputName=UVScale
     input {expression=&self:MaterialExpressionAppendVector_20}
    }
    {
     inputName=OriginalUV
     input {expression=&self:MaterialExpressionStaticSwitchParameter_0}
    }
   ]
  }
  n3 = AppendVector(A=Constant(), B=Divide(B=ScalarParameter() {defaultValue=10, group=Scifi_Screen, parameterName=LinesNumber}))
  n4 = ScalarParameter() {defaultValue=-1, group=Scifi_Screen, parameterName=SpeedDirection01}
  n5 = Multiply(A=Time(), B=ScalarParameter() {defaultValue=1, group=Scifi_Screen, parameterName=TextSpeed01})
  n6 = Divide(A=@n5, B=MaterialFunctionCall() {materialFunction=&Pi}.Result)
  n7 = Add(A=Multiply(A=Sine(None=@n6)) {constB=0.5}) {constB=0.5}
  n8 = Add(A=Add(A=@n5, B=Multiply(B=@n7) {constA=1}), B=ScalarParameter() {parameterName=TimeOffset01})
  n9 = Add(
   A=NamedRerouteDeclaration(None=@n2) {name=AdjustUV, nodeColor=<rgba 0 1 0.5312 1>}
   B=Multiply(A=Multiply(A=@n3, B=@n4) {constB=-1}, B=Floor(None=@n8))
  )
  Frac(None=@n9)
  n10 = MaterialFunctionCall() {materialFunction=&MF_CharacterParent_2023_PreFX}
  n11 = GetMaterialAttributes(None=@n10.Result) {attributeGetTypes [#1]}
  n12 = NamedRerouteUsage()
  n13 = TextureSampleParameter2D() {
   group=Scifi_Screen
   parameterName=TechArtMask_RE
   samplerType=SAMPLERTYPE_Masks
   texture=/BRCosmetics/Characters/Player/Male/Medium/Bodies/M_MED_HydroBottle/Textures/T_Hydrobottle_Body_Techart
  }
  n14 = ChannelMaskParameter(None=@n13.RGB) {group=Scifi_Screen, parameterName=ScreenChannel01}
  n15 = ChannelMaskParameter(None=@n13.RGB) {group=Scifi_Screen, maskChannel=Green, parameterName=ScreenChannel02, defaultValue=<rgba 0 1 0 0>}
  n16 = SmoothStep(
   Max=ScalarParameter() {defaultValue=0.1, group=Scifi_Screen, parameterName=MaxAmount_TechArtMask}
   Value=Add(
    A=NamedRerouteDeclaration(None=@n14.R) {name=TechArtMask_01, nodeColor=<rgba 0 0.8359 1 1>}
    B=NamedRerouteDeclaration(None=@n15.R) {name=TechArtMask_02, nodeColor=<rgba 0 0.8359 1 1>}
   )
  ) {constMax=0.2}
  MaterialFunctionCall(
   Material=SetMaterialAttributes(MaterialAttributes=@n11.MaterialAttributes, Emissive Color=LinearInterpolate(A=@n11.Emissive Color, B=@n12, Alpha=@n16)) {attributeSetTypes [#1]}
   M=@n10.M
   Pre Skinned Local Position=@n10.M
   Distance Blend=@n10.M
   Fresnel=@n10.M
  ) {materialFunction=&MF_CharacterParent_2023_PostFX}
  NamedRerouteDeclaration(None=@n1.ScaleY) {name=ScaleY, nodeColor=<rgba 0.2422 0 1 1>}
  n17 = NamedRerouteUsage()
  n18 = Constant() {r=1}
  NamedRerouteDeclaration(None=If(
    A=ComponentMask(None=@n17) {g=true}
    B=ScalarParameter() {group=Scifi_Screen, parameterName=BottomLine}
    "A > B"=Constant()
    "A == B"=@n18
    "A < B"=@n18
   )) {name=BottomMask, nodeColor=<rgba 0.4531 1 0 1>}
  n19 = NamedRerouteUsage()
  n20 = Step(X=@n19) {constY=0.85}
  NamedRerouteDeclaration(None=@n20) {name=Area01, nodeColor=<rgba 0.5 1 0 1>}
  n21 = VectorParameter() {group=Scifi_Screen, parameterName=BackgroundDarkRed, defaultValue=<rgba 0.8388 0.3596 0 1>}
  n22 = VectorParameter() {group=Scifi_Screen, parameterName=BackGroundRed, defaultValue=<rgba 0.8388 0.5354 0.0994 1>}
  n23 = TextureSampleParameter2D(UVs=Multiply(A=NamedRerouteUsage(), B=ScalarParameter() {defaultValue=1, group=Scifi_Screen, parameterName=ScanLineTextureTilling01})) {
   group=Scifi_Screen
   parameterName=ScanLine
   samplerSource=SSM_Wrap_WorldGroupSettings
   texture=/BRCosmetics/Weapons/FORT_Melee/Pickaxe_HydroBottle/Textures/T_ScanLines
  }
  n24 = VectorParameter() {group=Scifi_Screen, parameterName=TextColorRed, defaultValue=<rgba 0.8388 0.5354 0.0994 1>}
  n25 = TextureObjectParameter() {
   group=Scifi_Screen
   parameterName=LeftText
   samplerType=SAMPLERTYPE_Masks
   texture=/BRCosmetics/Accessories/FORT_Backpacks/Backpack_M_MED_HydroBottle/Textures/T_ScreenTilable_Mask_04_Left
  }
  n26 = TextureSample(UVs=@n9, Tex=@n25) {
   samplerSource=SSM_Wrap_WorldGroupSettings
   samplerType=SAMPLERTYPE_Masks
   texture=&T_ScreenTilable_Mask_04_Left
  }
  n27 = LinearInterpolate(A=@n26.R, B=Multiply(A=@n26.R, B=NamedRerouteUsage()), Alpha=NamedRerouteUsage())
  n28 = Add(
   A=NamedRerouteDeclaration(None=LinearInterpolate(A=@n21.RGB, B=@n22.RGB, Alpha=@n23.R)) {name=ScreenBackGround, nodeColor=<rgba 1 0.8438 0 1>}
   B=Multiply(A=@n24.RGB, B=Add(A=@n27, B=@n26.G))
  )
  n29 = Add(A=NamedRerouteUsage(), B=@n28)
  NamedRerouteDeclaration(None=Multiply(A=StaticSwitchParameter(True=@n29, False=@n28) {group=Scifi_Screen, parameterName=DebugArea}, B=NamedRerouteUsage())) {name=ScreenColor01, nodeColor=<rgba 1 0.4453 0 1>}
  n30 = Step(X=@n19) {constY=0.7}
  NamedRerouteDeclaration(None=Multiply(A=OneMinus(None=@n20), B=@n30)) {name=Area02, nodeColor=<rgba 0.125 0 1 1>}
  n31 = NamedRerouteUsage()
  n32 = ComponentMask(None=@n31) {r=true}
  n33 = VectorParameter() {group=Scifi_Screen, parameterName=TextBoundingBox02, channelNames {a=YMax, b=YMin, g=XMax, r=XMin}, defaultValue=<rgba 0 1 0 1>}
  n34 = Constant()
  n35 = If(
   A=@n32
   B=NamedRerouteDeclaration(None=@n33.XMin) {name=XMin_1, nodeColor=<rgba 0.2891 0 1 1>}
   "A > B"=Constant() {r=1}
   "A == B"=@n34
   "A < B"=@n34
  )
  n36 = Constant()
  n37 = If(A=@n32, B=NamedRerouteUsage(), "A > B"=@n36, "A == B"=@n36, "A < B"=Constant() {r=1})
  n38 = ComponentMask(None=@n31) {g=true}
  n39 = Constant()
  n40 = Constant()
  n41 = Multiply(
   A=If(A=@n38, B=@n33.YMin, "A > B"=Constant() {r=1}, "A == B"=@n39, "A < B"=@n39)
   B=If(A=@n38, B=@n33.YMax, "A > B"=@n40, "A == B"=@n40, "A < B"=Constant() {r=1})
  )
  NamedRerouteDeclaration(None=Multiply(A=Multiply(A=@n35, B=@n37), B=@n41)) {name=AreaMaskDynamic_1, nodeColor=<rgba 1 0.2109 0 1>}
  NamedRerouteDeclaration(None=@n33.XMax) {name=XMax_1, nodeColor=<rgba 0.2891 0 1 1>}
  n42 = Multiply(A=Time(), B=ScalarParameter() {defaultValue=1, group=Scifi_Screen, parameterName=TextSpeed02})
  n43 = Divide(A=@n42, B=MaterialFunctionCall() {materialFunction=&Pi}.Result)
  n44 = Add(A=Multiply(A=Sine(None=@n43)) {constB=0.5}) {constB=0.5}
  n45 = Add(A=Add(A=@n42, B=Multiply(B=@n44) {constA=1}), B=ScalarParameter() {parameterName=TimeOffset02})
  NamedRerouteDeclaration(None=@n45) {name=TimeAdj_1, nodeColor=<rgba 1 0 0.3516 1>}
  n46 = ScalarParameter() {defaultValue=20, parameterName=HorizontalDivisor}
  n47 = Multiply(A=NamedRerouteUsage(), B=@n46)
  n48 = Divide(A=Fmod(A=Floor(None=@n47), B=@n46), B=@n46)
  NamedRerouteDeclaration(None=LinearInterpolate(A=NamedRerouteUsage(), B=NamedRerouteUsage(), Alpha=@n48)) {name=XMax_Line_1, nodeColor=<rgba 0 0.6719 1 1>}
  n49 = Constant()
  n50 = If(A=@n32, B=NamedRerouteUsage(), "A > B"=@n49, "A == B"=@n49, "A < B"=Constant() {r=1})
  NamedRerouteDeclaration(None=Multiply(A=Multiply(A=@n35, B=@n50), B=@n41)) {name=AreaMaskStatic_1, nodeColor=<rgba 0.7578 1 0 1>}
  n51 = VectorParameter() {
   group=Scifi_Screen
   parameterName=TextTexture_Center02
   channelNames {a=ScaleY, b=ScaleX, g=CenterY, r=CenterX}
   defaultValue=<rgba 0.5 0.5 1 1>
  }
  NamedRerouteDeclaration(None=@n51.ScaleY) {name=ScaleY_1, nodeColor=<rgba 0.2422 0 1 1>}
  n52 = NamedRerouteUsage()
  n53 = Constant() {r=1}
  NamedRerouteDeclaration(None=If(
    A=ComponentMask(None=@n52) {g=true}
    B=ScalarParameter() {group=Scifi_Screen, parameterName=BottomLine}
    "A > B"=Constant()
    "A == B"=@n53
    "A < B"=@n53
   )) {name=BottomMask_1, nodeColor=<rgba 0.4531 1 0 1>}
  n54 = VectorParameter() {group=Scifi_Screen, parameterName=BackgroundDarkOrange, defaultValue=<rgba 0.8388 0.3596 0 1>}
  n55 = VectorParameter() {group=Scifi_Screen, parameterName=BackGroundYellow, defaultValue=<rgba 0.8388 0.5354 0.0994 1>}
  n56 = TextureSampleParameter2D(UVs=Multiply(A=NamedRerouteUsage(), B=ScalarParameter() {defaultValue=0.5, group=Scifi_Screen, parameterName=ScanLineTextureTilling02})) {
   group=Scifi_Screen
   parameterName=ScanLine
   samplerSource=SSM_Wrap_WorldGroupSettings
   texture=&T_ScanLines
  }
  n57 = VectorParameter() {group=Scifi_Screen, parameterName=TextColor_Yellow, defaultValue=<rgba 0.8388 0.5354 0.0994 1>}
  MaterialExpressionAppendVector_1 = AppendVector(A=@n51.CenterX, B=@n51.CenterX)
  MaterialExpressionAppendVector_2 = AppendVector(A=@n51.ScaleX, B=@n51.ScaleX)
  MaterialExpressionNamedRerouteUsage_31 = NamedRerouteUsage()
  n58 = Custom(UVPoint=@MaterialExpressionAppendVector_1, UVScale=@MaterialExpressionAppendVector_2, OriginalUV=@MaterialExpressionNamedRerouteUsage_31) {
   code="float2 DistUV =OriginalUV-UVPoint;\r\nDistUV = DistUV/UVScale + float2(0.5,0.5);\r\n\r\nreturn DistUV;"
   description=CopyToPointUV
   outputType=CMOT_Float2
   inputs [
    {
     inputName=UVPoint
     input {expression=&self:MaterialExpressionAppendVector_1}
    }
    {
     inputName=UVScale
     input {expression=&self:MaterialExpressionAppendVector_2}
    }
    {
     inputName=OriginalUV
     input {expression=&self:MaterialExpressionNamedRerouteUsage_31}
    }
   ]
  }
  n59 = ScalarParameter() {defaultValue=10, group=Scifi_Screen, parameterName=LinesNumber}
  n60 = Multiply(A=AppendVector(A=Constant(), B=Divide(B=@n59)), B=ScalarParameter() {defaultValue=-1, group=Scifi_Screen, parameterName=SpeedDirection02}) {constB=-1}
  n61 = Floor(None=@n45)
  n62 = TextureObjectParameter() {
   group=Scifi_Screen
   parameterName=LeftText
   samplerType=SAMPLERTYPE_Masks
   texture=&T_ScreenTilable_Mask_04_Left
  }
  n63 = TextureSample(UVs=Add(A=NamedRerouteDeclaration(None=@n58) {name=AdjustUV_1, nodeColor=<rgba 0 1 0.5312 1>}, B=Multiply(A=@n60, B=@n61)), Tex=@n62) {
   samplerSource=SSM_Wrap_WorldGroupSettings
   samplerType=SAMPLERTYPE_Masks
   texture=&T_ScreenTilable_Mask_04_Left
  }
  n64 = LinearInterpolate(A=@n63.R, B=Multiply(A=@n63.R, B=NamedRerouteUsage()), Alpha=NamedRerouteUsage())
  n65 = Add(
   A=NamedRerouteDeclaration(None=LinearInterpolate(A=@n54.RGB, B=@n55.RGB, Alpha=@n56.R)) {name=ScreenBackGround_Yellow, nodeColor=<rgba 1 0.8438 0 1>}
   B=Multiply(A=@n57.RGB, B=Add(A=@n64, B=@n63.G))
  )
  n66 = Add(A=NamedRerouteUsage(), B=@n65)
  NamedRerouteDeclaration(None=Multiply(A=StaticSwitchParameter(True=@n66, False=@n65) {group=Scifi_Screen, parameterName=DebugArea}, B=NamedRerouteUsage())) {name=ScreenColor02, nodeColor=<rgba 1 0.4453 0 1>}
  NamedRerouteDeclaration(None=@MaterialExpressionStaticSwitchParameter_0) {name=BaseCoord, nodeColor=<rgba 1 0.1641 0 1>}
  n67 = Step(X=@n19) {constY=0.5}
  NamedRerouteDeclaration(None=Multiply(A=OneMinus(None=@n30), B=@n67)) {name=Area03, nodeColor=<rgba 0 0.5781 1 1>}
  n68 = NamedRerouteUsage()
  n69 = NamedRerouteUsage()
  n70 = Add(A=Add(A=Add(A=@n68, B=@n69), B=NamedRerouteUsage()), B=NamedRerouteUsage())
  n71 = NamedRerouteUsage()
  NamedRerouteDeclaration(None=Add(A=NamedRerouteUsage(), B=Add(A=@n70, B=@n71))) {name=Screen_Typing, nodeColor=<rgba 1 0.7969 0 1>}
  n72 = TextureSampleParameter2D(UVs=Multiply(A=NamedRerouteUsage(), B=ScalarParameter() {defaultValue=1, group=Scifi_Screen, parameterName=ScanLineTextureTilling_Graph})) {
   group=Scifi_Screen
   parameterName=ScanLine
   samplerSource=SSM_Wrap_WorldGroupSettings
   texture=&T_ScanLines
  }
  NamedRerouteDeclaration(None=@n72.R) {name=GraphBackGround, nodeColor=<rgba 0.1016 0 1 1>}
  n73 = NamedRerouteUsage()
  n74 = ComponentMask(None=@n73) {r=true}
  n75 = VectorParameter() {group=Scifi_Screen, parameterName=TextBoundingBox03, channelNames {a=YMax, b=YMin, g=XMax, r=XMin}, defaultValue=<rgba 0 1 0 1>}
  n76 = Constant()
  n77 = If(
   A=@n74
   B=NamedRerouteDeclaration(None=@n75.XMin) {name=XMin_2, nodeColor=<rgba 0.2891 0 1 1>}
   "A > B"=Constant() {r=1}
   "A == B"=@n76
   "A < B"=@n76
  )
  n78 = Constant()
  n79 = If(A=@n74, B=NamedRerouteUsage(), "A > B"=@n78, "A == B"=@n78, "A < B"=Constant() {r=1})
  n80 = ComponentMask(None=@n73) {g=true}
  n81 = Constant()
  n82 = Constant()
  n83 = Multiply(
   A=If(A=@n80, B=@n75.YMin, "A > B"=Constant() {r=1}, "A == B"=@n81, "A < B"=@n81)
   B=If(A=@n80, B=@n75.YMax, "A > B"=@n82, "A == B"=@n82, "A < B"=Constant() {r=1})
  )
  NamedRerouteDeclaration(None=Multiply(A=Multiply(A=@n77, B=@n79), B=@n83)) {name=AreaMaskDynamic_2, nodeColor=<rgba 1 0.2109 0 1>}
  NamedRerouteDeclaration(None=@n75.XMax) {name=XMax_2, nodeColor=<rgba 0.2891 0 1 1>}
  n84 = Multiply(A=Time(), B=ScalarParameter() {defaultValue=1, group=Scifi_Screen, parameterName=TextSpeed03})
  n85 = Divide(A=@n84, B=MaterialFunctionCall() {materialFunction=&Pi}.Result)
  n86 = Add(A=Multiply(A=Sine(None=@n85)) {constB=0.5}) {constB=0.5}
  n87 = Add(A=Add(A=@n84, B=Multiply(B=@n86) {constA=1}), B=ScalarParameter() {parameterName=TimeOffset02})
  NamedRerouteDeclaration(None=@n87) {name=TimeAdj_2, nodeColor=<rgba 1 0 0.3516 1>}
  n88 = ScalarParameter() {defaultValue=20, parameterName=HorizontalDivisor}
  n89 = Multiply(A=NamedRerouteUsage(), B=@n88)
  n90 = Divide(A=Fmod(A=Floor(None=@n89), B=@n88), B=@n88)
  NamedRerouteDeclaration(None=LinearInterpolate(A=NamedRerouteUsage(), B=NamedRerouteUsage(), Alpha=@n90)) {name=XMax_Line_2, nodeColor=<rgba 0 0.6719 1 1>}
  n91 = Constant()
  n92 = If(A=@n74, B=NamedRerouteUsage(), "A > B"=@n91, "A == B"=@n91, "A < B"=Constant() {r=1})
  NamedRerouteDeclaration(None=Multiply(A=Multiply(A=@n77, B=@n92), B=@n83)) {name=AreaMaskStatic_2, nodeColor=<rgba 0.7578 1 0 1>}
  n93 = VectorParameter() {
   group=Scifi_Screen
   parameterName=TextTexture_Center03
   channelNames {a=ScaleY, b=ScaleX, g=CenterY, r=CenterX}
   defaultValue=<rgba 0.5 0.5 1 1>
  }
  NamedRerouteDeclaration(None=@n93.ScaleY) {name=ScaleY_2, nodeColor=<rgba 0.2422 0 1 1>}
  n94 = NamedRerouteUsage()
  n95 = Constant() {r=1}
  NamedRerouteDeclaration(None=If(
    A=ComponentMask(None=@n94) {g=true}
    B=ScalarParameter() {group=Scifi_Screen, parameterName=BottomLine}
    "A > B"=Constant()
    "A == B"=@n95
    "A < B"=@n95
   )) {name=BottomMask_2, nodeColor=<rgba 0.4531 1 0 1>}
  NamedRerouteDeclaration(None=@n57.RGB) {name=Text_Yellow, nodeColor=<rgba 0.3594 1 0 1>}
  n96 = NamedRerouteUsage()
  n97 = ComponentMask(None=@n96) {r=true}
  n98 = VectorParameter() {group=Scifi_Screen, parameterName=TextBoundingBox01, channelNames {a=YMax, b=YMin, g=XMax, r=XMin}, defaultValue=<rgba 0 1 0 1>}
  n99 = Constant()
  n100 = If(
   A=@n97
   B=NamedRerouteDeclaration(None=@n98.XMin) {name=XMin, nodeColor=<rgba 0.2891 0 1 1>}
   "A > B"=Constant() {r=1}
   "A == B"=@n99
   "A < B"=@n99
  )
  n101 = Constant()
  n102 = If(A=@n97, B=NamedRerouteUsage(), "A > B"=@n101, "A == B"=@n101, "A < B"=Constant() {r=1})
  n103 = ComponentMask(None=@n96) {g=true}
  n104 = Constant()
  n105 = Constant()
  n106 = Multiply(
   A=If(A=@n103, B=@n98.YMin, "A > B"=Constant() {r=1}, "A == B"=@n104, "A < B"=@n104)
   B=If(A=@n103, B=@n98.YMax, "A > B"=@n105, "A == B"=@n105, "A < B"=Constant() {r=1})
  )
  NamedRerouteDeclaration(None=Multiply(A=Multiply(A=@n100, B=@n102), B=@n106)) {name=AreaMaskDynamic, nodeColor=<rgba 1 0.2109 0 1>}
  n107 = TextureSampleParameter2D(UVs=Multiply(A=NamedRerouteUsage(), B=ScalarParameter() {defaultValue=0.5, group=Scifi_Screen, parameterName=ScanLineTextureTilling03})) {
   group=Scifi_Screen
   parameterName=ScanLine
   samplerSource=SSM_Wrap_WorldGroupSettings
   texture=&T_ScanLines
  }
  MaterialExpressionAppendVector_4 = AppendVector(A=@n93.CenterX, B=@n93.CenterX)
  MaterialExpressionAppendVector_5 = AppendVector(A=@n93.ScaleX, B=@n93.ScaleX)
  MaterialExpressionNamedRerouteUsage_42 = NamedRerouteUsage()
  n108 = Custom(UVPoint=@MaterialExpressionAppendVector_4, UVScale=@MaterialExpressionAppendVector_5, OriginalUV=@MaterialExpressionNamedRerouteUsage_42) {
   code="float2 DistUV =OriginalUV-UVPoint;\r\nDistUV = DistUV/UVScale + float2(0.5,0.5);\r\n\r\nreturn DistUV;"
   description=CopyToPointUV
   outputType=CMOT_Float2
   inputs [
    {
     inputName=UVPoint
     input {expression=&self:MaterialExpressionAppendVector_4}
    }
    {
     inputName=UVScale
     input {expression=&self:MaterialExpressionAppendVector_5}
    }
    {
     inputName=OriginalUV
     input {expression=&self:MaterialExpressionNamedRerouteUsage_42}
    }
   ]
  }
  n109 = ScalarParameter() {defaultValue=10, group=Scifi_Screen, parameterName=LinesNumber}
  n110 = Multiply(A=AppendVector(A=Constant(), B=Divide(B=@n109)), B=ScalarParameter() {defaultValue=-1, group=Scifi_Screen, parameterName=SpeedDirection03}) {constB=-1}
  n111 = Floor(None=@n87)
  n112 = TextureObjectParameter() {
   group=Scifi_Screen
   parameterName=LeftText
   samplerType=SAMPLERTYPE_Masks
   texture=&T_ScreenTilable_Mask_04_Left
  }
  n113 = TextureSample(UVs=Add(A=NamedRerouteDeclaration(None=@n108) {name=AdjustUV_2, nodeColor=<rgba 0 1 0.5312 1>}, B=Multiply(A=@n110, B=@n111)), Tex=@n112) {
   samplerSource=SSM_Wrap_WorldGroupSettings
   samplerType=SAMPLERTYPE_Masks
   texture=&T_ScreenTilable_Mask_04_Left
  }
  n114 = LinearInterpolate(A=@n113.R, B=Multiply(A=@n113.R, B=NamedRerouteUsage()), Alpha=NamedRerouteUsage())
  n115 = Add(
   A=LinearInterpolate(A=NamedRerouteUsage(), B=NamedRerouteUsage(), Alpha=@n107.R)
   B=Multiply(A=NamedRerouteUsage(), B=Add(A=@n114, B=@n113.G))
  )
  n116 = Add(A=NamedRerouteUsage(), B=@n115)
  NamedRerouteDeclaration(None=Multiply(A=StaticSwitchParameter(True=@n116, False=@n115) {group=Scifi_Screen, parameterName=DebugArea}, B=NamedRerouteUsage())) {name=ScreenColor03, nodeColor=<rgba 1 0.4453 0 1>}
  n117 = NamedRerouteUsage()
  n118 = Step(X=@n117) {constY=0.9}
  n119 = Step(X=@n117) {constY=0.6}
  n120 = Step(X=@n117) {constY=0.4}
  n121 = Multiply(A=OneMinus(None=@n119), B=@n120)
  NamedRerouteDeclaration(None=Add(
    A=NamedRerouteDeclaration(None=@n118) {name=Graph01, nodeColor=<rgba 0.9922 1 0 1>}
    B=NamedRerouteDeclaration(None=@n121) {name=Graph03, nodeColor=<rgba 0.9922 1 0 1>}
   )) {name=YellowGraph, nodeColor=<rgba 1 0 0.9844 1>}
  NamedRerouteDeclaration(None=@n21.RGB) {name=RedDarkColor, nodeColor=<rgba 0 1 1 1>}
  NamedRerouteDeclaration(None=@n22.RGB) {name=RedLightColor, nodeColor=<rgba 1 0.6797 0 1>}
  NamedRerouteDeclaration(None=@n24.RGB) {name=Text_Red, nodeColor=<rgba 0.3594 0 1 1>}
  n122 = Step(X=@n19) {constY=0.3}
  NamedRerouteDeclaration(None=Multiply(A=OneMinus(None=@n67), B=@n122)) {name=Area04, nodeColor=<rgba 0 0.5781 1 1>}
  n123 = NamedRerouteUsage()
  n124 = ComponentMask(None=@n123) {r=true}
  n125 = VectorParameter() {group=Scifi_Screen, parameterName=TextBoundingBox04, channelNames {a=YMax, b=YMin, g=XMax, r=XMin}, defaultValue=<rgba 0 1 0 1>}
  n126 = Constant()
  n127 = If(
   A=@n124
   B=NamedRerouteDeclaration(None=@n125.XMin) {name=XMin_3, nodeColor=<rgba 0.2891 0 1 1>}
   "A > B"=Constant() {r=1}
   "A == B"=@n126
   "A < B"=@n126
  )
  n128 = Constant()
  n129 = OneMinus(None=If(A=@n124, B=NamedRerouteUsage(), "A > B"=@n128, "A == B"=@n128, "A < B"=Constant() {r=1}))
  n130 = ComponentMask(None=@n123) {g=true}
  n131 = Constant()
  n132 = Constant()
  n133 = Multiply(
   A=If(A=@n130, B=@n125.YMin, "A > B"=Constant() {r=1}, "A == B"=@n131, "A < B"=@n131)
   B=If(A=@n130, B=@n125.YMax, "A > B"=@n132, "A == B"=@n132, "A < B"=Constant() {r=1})
  )
  NamedRerouteDeclaration(None=Multiply(A=Multiply(A=@n127, B=@n129), B=@n133)) {name=AreaMaskDynamic_3, nodeColor=<rgba 1 0.2109 0 1>}
  NamedRerouteDeclaration(None=@n125.XMax) {name=XMax_3, nodeColor=<rgba 0.2891 0 1 1>}
  n134 = Multiply(A=Time(), B=ScalarParameter() {defaultValue=1, group=Scifi_Screen, parameterName=TextSpeed04})
  n135 = Divide(A=@n134, B=MaterialFunctionCall() {materialFunction=&Pi}.Result)
  n136 = Add(A=Multiply(A=Sine(None=@n135)) {constB=0.5}) {constB=0.5}
  n137 = Add(A=Add(A=@n134, B=Multiply(B=@n136) {constA=1}), B=ScalarParameter() {parameterName=TimeOffset04})
  NamedRerouteDeclaration(None=@n137) {name=TimeAdj_3, nodeColor=<rgba 1 0 0.3516 1>}
  n138 = ScalarParameter() {defaultValue=20, parameterName=HorizontalDivisor}
  n139 = Multiply(A=NamedRerouteUsage(), B=@n138)
  n140 = Divide(A=Fmod(A=Floor(None=@n139), B=@n138), B=@n138)
  NamedRerouteDeclaration(None=LinearInterpolate(A=NamedRerouteUsage(), B=NamedRerouteUsage(), Alpha=OneMinus(None=@n140))) {name=XMax_Line_3, nodeColor=<rgba 0 0.6719 1 1>}
  n141 = Constant()
  n142 = If(A=@n124, B=NamedRerouteUsage(), "A > B"=@n141, "A == B"=@n141, "A < B"=Constant() {r=1})
  NamedRerouteDeclaration(None=Multiply(A=Multiply(A=@n127, B=@n142), B=@n133)) {name=AreaMaskStatic_3, nodeColor=<rgba 0.7578 1 0 1>}
  n143 = VectorParameter() {
   group=Scifi_Screen
   parameterName=TextTexture_Center04
   channelNames {a=ScaleY, b=ScaleX, g=CenterY, r=CenterX}
   defaultValue=<rgba 0.5 0.5 1 1>
  }
  NamedRerouteDeclaration(None=@n143.ScaleY) {name=ScaleY_3, nodeColor=<rgba 0.2422 0 1 1>}
  n144 = NamedRerouteUsage()
  n145 = Constant() {r=1}
  NamedRerouteDeclaration(None=If(
    A=ComponentMask(None=@n144) {g=true}
    B=ScalarParameter() {group=Scifi_Screen, parameterName=BottomLine}
    "A > B"=Constant()
    "A == B"=@n145
    "A < B"=@n145
   )) {name=BottomMask_3, nodeColor=<rgba 0.4531 1 0 1>}
  n146 = TextureSampleParameter2D(UVs=Multiply(A=NamedRerouteUsage(), B=ScalarParameter() {defaultValue=0.5, group=Scifi_Screen, parameterName=ScanLineTextureTilling04})) {
   group=Scifi_Screen
   parameterName=ScanLine
   samplerSource=SSM_Wrap_WorldGroupSettings
   texture=&T_ScanLines
  }
  MaterialExpressionAppendVector_7 = AppendVector(A=@n143.CenterX, B=@n143.CenterX)
  MaterialExpressionAppendVector_8 = AppendVector(A=@n143.ScaleX, B=@n143.ScaleX)
  MaterialExpressionNamedRerouteUsage_62 = NamedRerouteUsage()
  n147 = Custom(UVPoint=@MaterialExpressionAppendVector_7, UVScale=@MaterialExpressionAppendVector_8, OriginalUV=@MaterialExpressionNamedRerouteUsage_62) {
   code="float2 DistUV =OriginalUV-UVPoint;\r\nDistUV = DistUV/UVScale + float2(0.5,0.5);\r\n\r\nreturn DistUV;"
   description=CopyToPointUV
   outputType=CMOT_Float2
   inputs [
    {
     inputName=UVPoint
     input {expression=&self:MaterialExpressionAppendVector_7}
    }
    {
     inputName=UVScale
     input {expression=&self:MaterialExpressionAppendVector_8}
    }
    {
     inputName=OriginalUV
     input {expression=&self:MaterialExpressionNamedRerouteUsage_62}
    }
   ]
  }
  n148 = ScalarParameter() {defaultValue=10, group=Scifi_Screen, parameterName=LinesNumber}
  n149 = Multiply(A=AppendVector(A=Constant(), B=Divide(B=@n148)), B=ScalarParameter() {defaultValue=-1, group=Scifi_Screen, parameterName=SpeedDirection04}) {constB=-1}
  n150 = Floor(None=@n137)
  n151 = TextureObjectParameter() {
   group=Scifi_Screen
   parameterName=RightText
   samplerType=SAMPLERTYPE_Masks
   texture=/BRCosmetics/Weapons/FORT_Melee/Pickaxe_HydroBottle/Textures/T_ScreenTilable
  }
  n152 = TextureSample(UVs=Add(A=NamedRerouteDeclaration(None=@n147) {name=AdjustUV_3, nodeColor=<rgba 0 1 0.5312 1>}, B=Multiply(A=@n149, B=@n150)), Tex=@n151) {
   samplerSource=SSM_Wrap_WorldGroupSettings
   samplerType=SAMPLERTYPE_Masks
   texture=&T_ScreenTilable_Mask_04_Left
  }
  n153 = LinearInterpolate(A=@n152.R, B=Multiply(A=@n152.R, B=NamedRerouteUsage()), Alpha=NamedRerouteUsage())
  n154 = Add(
   A=LinearInterpolate(A=NamedRerouteUsage(), B=NamedRerouteUsage(), Alpha=@n146.R)
   B=Multiply(A=NamedRerouteUsage(), B=Add(A=@n153, B=@n152.G))
  )
  n155 = Add(A=NamedRerouteUsage(), B=@n154)
  NamedRerouteDeclaration(None=Multiply(A=StaticSwitchParameter(True=@n155, False=@n154) {group=Scifi_Screen, parameterName=DebugArea}, B=NamedRerouteUsage())) {name=ScreenColor04, nodeColor=<rgba 1 0.4453 0 1>}
  NamedRerouteDeclaration(None=@n54.RGB) {name=DarkYellow, nodeColor=<rgba 1 0 0.5625 1>}
  NamedRerouteDeclaration(None=@n55.RGB) {name=LightYellow, nodeColor=<rgba 0 0.8828 1 1>}
  n156 = NamedRerouteUsage()
  n157 = ComponentMask(None=@n156) {r=true}
  n158 = VectorParameter() {group=Scifi_Screen, parameterName=TextBoundingBox05, channelNames {a=YMax, b=YMin, g=XMax, r=XMin}, defaultValue=<rgba 0 1 0 1>}
  n159 = Constant()
  n160 = If(
   A=@n157
   B=NamedRerouteDeclaration(None=@n158.XMin) {name=XMin_4, nodeColor=<rgba 0.2891 0 1 1>}
   "A > B"=Constant() {r=1}
   "A == B"=@n159
   "A < B"=@n159
  )
  n161 = Constant()
  n162 = OneMinus(None=If(A=@n157, B=NamedRerouteUsage(), "A > B"=@n161, "A == B"=@n161, "A < B"=Constant() {r=1}))
  n163 = ComponentMask(None=@n156) {g=true}
  n164 = Constant()
  n165 = Constant()
  n166 = Multiply(
   A=If(A=@n163, B=@n158.YMin, "A > B"=Constant() {r=1}, "A == B"=@n164, "A < B"=@n164)
   B=If(A=@n163, B=@n158.YMax, "A > B"=@n165, "A == B"=@n165, "A < B"=Constant() {r=1})
  )
  NamedRerouteDeclaration(None=Multiply(A=Multiply(A=@n160, B=@n162), B=@n166)) {name=AreaMaskDynamic_4, nodeColor=<rgba 1 0.2109 0 1>}
  NamedRerouteDeclaration(None=@n98.XMax) {name=XMax, nodeColor=<rgba 0.2891 0 1 1>}
  NamedRerouteDeclaration(None=@n158.XMax) {name=XMax_4, nodeColor=<rgba 0.2891 0 1 1>}
  n167 = Multiply(A=Time(), B=ScalarParameter() {defaultValue=1, group=Scifi_Screen, parameterName=TextSpeed05})
  n168 = Divide(A=@n167, B=MaterialFunctionCall() {materialFunction=&Pi}.Result)
  n169 = Add(A=Multiply(A=Sine(None=@n168)) {constB=0.5}) {constB=0.5}
  n170 = Add(A=Add(A=@n167, B=Multiply(B=@n169) {constA=1}), B=ScalarParameter() {parameterName=TimeOffset05})
  NamedRerouteDeclaration(None=@n170) {name=TimeAdj_4, nodeColor=<rgba 1 0 0.3516 1>}
  n171 = ScalarParameter() {defaultValue=20, parameterName=HorizontalDivisor}
  n172 = Multiply(A=NamedRerouteUsage(), B=@n171)
  n173 = Divide(A=Fmod(A=Floor(None=@n172), B=@n171), B=@n171)
  NamedRerouteDeclaration(None=LinearInterpolate(A=NamedRerouteUsage(), B=NamedRerouteUsage(), Alpha=OneMinus(None=@n173))) {name=XMax_Line_4, nodeColor=<rgba 0 0.6719 1 1>}
  n174 = Constant()
  n175 = If(A=@n157, B=NamedRerouteUsage(), "A > B"=@n174, "A == B"=@n174, "A < B"=Constant() {r=1})
  NamedRerouteDeclaration(None=Multiply(A=Multiply(A=@n160, B=@n175), B=@n166)) {name=AreaMaskStatic_4, nodeColor=<rgba 0.7578 1 0 1>}
  n176 = VectorParameter() {
   group=Scifi_Screen
   parameterName=TextTexture_Center05
   channelNames {a=ScaleY, b=ScaleX, g=CenterY, r=CenterX}
   defaultValue=<rgba 0.5 0.5 1 1>
  }
  NamedRerouteDeclaration(None=@n176.ScaleY) {name=ScaleY_4, nodeColor=<rgba 0.2422 0 1 1>}
  n177 = NamedRerouteUsage()
  n178 = Constant() {r=1}
  NamedRerouteDeclaration(None=If(
    A=ComponentMask(None=@n177) {g=true}
    B=ScalarParameter() {group=Scifi_Screen, parameterName=BottomLine}
    "A > B"=Constant()
    "A == B"=@n178
    "A < B"=@n178
   )) {name=BottomMask_4, nodeColor=<rgba 0.4531 1 0 1>}
  n179 = TextureSampleParameter2D(UVs=Multiply(A=NamedRerouteUsage(), B=ScalarParameter() {defaultValue=0.5, group=Scifi_Screen, parameterName=ScanLineTextureTilling05})) {
   group=Scifi_Screen
   parameterName=ScanLine
   samplerSource=SSM_Wrap_WorldGroupSettings
   texture=&T_ScanLines
  }
  MaterialExpressionAppendVector_10 = AppendVector(A=@n176.CenterX, B=@n176.CenterX)
  MaterialExpressionAppendVector_11 = AppendVector(A=@n176.ScaleX, B=@n176.ScaleX)
  MaterialExpressionNamedRerouteUsage_83 = NamedRerouteUsage()
  n180 = Custom(UVPoint=@MaterialExpressionAppendVector_10, UVScale=@MaterialExpressionAppendVector_11, OriginalUV=@MaterialExpressionNamedRerouteUsage_83) {
   code="float2 DistUV =OriginalUV-UVPoint;\r\nDistUV = DistUV/UVScale + float2(0.5,0.5);\r\n\r\nreturn DistUV;"
   description=CopyToPointUV
   outputType=CMOT_Float2
   inputs [
    {
     inputName=UVPoint
     input {expression=&self:MaterialExpressionAppendVector_10}
    }
    {
     inputName=UVScale
     input {expression=&self:MaterialExpressionAppendVector_11}
    }
    {
     inputName=OriginalUV
     input {expression=&self:MaterialExpressionNamedRerouteUsage_83}
    }
   ]
  }
  n181 = ScalarParameter() {defaultValue=10, group=Scifi_Screen, parameterName=LinesNumber}
  n182 = Multiply(A=AppendVector(A=Constant(), B=Divide(B=@n181)), B=ScalarParameter() {defaultValue=-1, group=Scifi_Screen, parameterName=SpeedDirection05}) {constB=-1}
  n183 = Floor(None=@n170)
  n184 = TextureObjectParameter() {
   group=Scifi_Screen
   parameterName=RightText
   samplerType=SAMPLERTYPE_Masks
   texture=&T_ScreenTilable
  }
  n185 = TextureSample(UVs=Add(A=NamedRerouteDeclaration(None=@n180) {name=AdjustUV_4, nodeColor=<rgba 0 1 0.5312 1>}, B=Multiply(A=@n182, B=@n183)), Tex=@n184) {
   samplerSource=SSM_Wrap_WorldGroupSettings
   samplerType=SAMPLERTYPE_Masks
   texture=&T_ScreenTilable_Mask_04_Left
  }
  n186 = LinearInterpolate(A=@n185.R, B=Multiply(A=@n185.R, B=NamedRerouteUsage()), Alpha=NamedRerouteUsage())
  n187 = Add(
   A=LinearInterpolate(A=NamedRerouteUsage(), B=NamedRerouteUsage(), Alpha=@n179.R)
   B=Multiply(A=NamedRerouteUsage(), B=Add(A=@n186, B=@n185.G))
  )
  n188 = Add(A=NamedRerouteUsage(), B=@n187)
  NamedRerouteDeclaration(None=Multiply(A=StaticSwitchParameter(True=@n188, False=@n187) {group=Scifi_Screen, parameterName=DebugArea}, B=NamedRerouteUsage())) {name=ScreenColor05, nodeColor=<rgba 1 0.4453 0 1>}
  NamedRerouteDeclaration(None=Multiply(A=OneMinus(None=@n122), B=Step(X=@n19) {constX=0.3, constY=0.1})) {name=Area05, nodeColor=<rgba 0 0.5781 1 1>}
  NamedRerouteDeclaration(None=@n8) {name=TimeAdj, nodeColor=<rgba 1 0 0.3516 1>}
  n189 = NamedRerouteUsage()
  n190 = LinearInterpolate(A=NamedRerouteUsage(), B=NamedRerouteUsage(), Alpha=@n72.R)
  n191 = NamedRerouteUsage()
  n192 = LinearInterpolate(A=NamedRerouteUsage(), B=NamedRerouteUsage(), Alpha=@n72.R)
  NamedRerouteDeclaration(None=Add(A=Multiply(A=@n189, B=@n190), B=Multiply(A=@n191, B=@n192))) {name=GraphBackGroundColor, nodeColor=<rgba 1 0.9844 0 1>}
  n193 = Multiply(A=OneMinus(None=@n118), B=@n119)
  n194 = Multiply(A=OneMinus(None=@n120), B=Step(X=@n117) {constY=0.2})
  NamedRerouteDeclaration(None=Add(
    A=NamedRerouteDeclaration(None=@n193) {name=Graph02, nodeColor=<rgba 0.9922 1 0 1>}
    B=NamedRerouteDeclaration(None=@n194) {name=Graph04, nodeColor=<rgba 0.9922 1 0 1>}
   )) {name=RedGraph, nodeColor=<rgba 0 1 0.6953 1>}
  n195 = VectorParameter() {
   group=Scifi_Screen_Graph
   parameterName=GraphTexture_Center01
   channelNames {a=ScaleY, b=ScaleX, g=CenterY, r=CenterX}
   defaultValue=<rgba 0.5 0.5 1 1>
  }
  NamedRerouteDeclaration(None=@n195.ScaleY) {name=ScaleY_5, nodeColor=<rgba 0.2422 0 1 1>}
  MaterialExpressionStaticSwitchParameter_6 = StaticSwitchParameter(True=TextureCoordinate() {coordinateIndex=1}, False=TextureCoordinate()) {defaultValue=true, group=Scifi_Screen_Graph, parameterName=UseTexcoord01}
  NamedRerouteDeclaration(None=@MaterialExpressionStaticSwitchParameter_6) {name=BaseCoord_Graph, nodeColor=<rgba 1 0.1641 0 1>}
  MaterialExpressionAppendVector_13 = AppendVector(A=@n195.CenterX, B=@n195.CenterX)
  MaterialExpressionAppendVector_14 = AppendVector(A=@n195.ScaleX, B=@n195.ScaleX)
  n196 = Custom(
   UVPoint=@MaterialExpressionAppendVector_13
   UVScale=@MaterialExpressionAppendVector_14
   OriginalUV=@MaterialExpressionStaticSwitchParameter_6
  ) {
   code="float2 DistUV =OriginalUV-UVPoint;\r\nDistUV = DistUV/UVScale + float2(0.5,0.5);\r\n\r\nreturn DistUV;"
   description=CopyToPointUV
   outputType=CMOT_Float2
   inputs [
    {
     inputName=UVPoint
     input {expression=&self:MaterialExpressionAppendVector_13}
    }
    {
     inputName=UVScale
     input {expression=&self:MaterialExpressionAppendVector_14}
    }
    {
     inputName=OriginalUV
     input {expression=&self:MaterialExpressionStaticSwitchParameter_6}
    }
   ]
  }
  n197 = VectorParameter() {group=Scifi_Screen_Graph, parameterName=GraphSpeed01, defaultValue=<rgba 0.1 0 0 0>}
  n198 = TextureObjectParameter() {
   group=Scifi_Screen_Graph
   parameterName=DownGraph
   samplerType=SAMPLERTYPE_Masks
   texture=/BRCosmetics/Characters/Player/Male/Medium/Bodies/M_MED_HydroBottle/Textures/T_HeartLine_Flip
  }
  n199 = TextureSample(UVs=Panner(Coordinate=@n196, Speed=AppendVector(A=@n197.R, B=@n197.R)), Tex=@n198) {
   samplerSource=SSM_Wrap_WorldGroupSettings
   samplerType=SAMPLERTYPE_Masks
   texture=/BRCosmetics/Characters/Player/Male/Medium/Bodies/M_MED_HydroBottle/Textures/T_HeartLine
  }
  n200 = Multiply(A=@n199.R, B=ScalarParameter() {defaultValue=0.3, group=Scifi_Screen_Graph, parameterName=BasicEM01})
  n201 = ComponentMask(None=@n196) {r=true}
  n202 = VectorParameter() {group=Scifi_Screen_Graph, parameterName=GraphBoundingBox01, channelNames {a=YMax, b=YMin, g=XMax, r=XMin}, defaultValue=<rgba 0 1 0 1>}
  n203 = Multiply(A=Time(), B=ScalarParameter() {defaultValue=0.5, group=Scifi_Screen_Graph, parameterName=GraphTimeSpeed01})
  n204 = LinearInterpolate(
   A=@n202.XMax
   B=@n202.XMax
   Alpha=StaticSwitchParameter(True=ScalarParameter() {group=Scifi_Screen_Graph, parameterName=00_GraphDebugProgress, sliderMax=1}, False=Frac(None=@n203)) {group=Scifi_Screen_Graph, parameterName=DebugGraph}
  )
  n205 = ScalarParameter() {defaultValue=0.3, group=Scifi_Screen_Graph, parameterName=GradientRadius01}
  n206 = ScalarParameter() {defaultValue=0.1, group=Scifi_Screen_Graph, parameterName=GradientRadius_Point01}
  n207 = Add(
   A=Multiply(
    A=SphereMask(A=@n201, B=@n204, Radius=@n205) {attenuationRadius=0.5, hardnessPercent=0.2}
    B=ScalarParameter() {defaultValue=10, group=Scifi_Screen_Graph, parameterName=GradientEM01}
   )
   B=Multiply(
    A=SphereMask(A=@n201, B=@n204, Radius=@n206) {attenuationRadius=0.5, hardnessPercent=10}
    B=ScalarParameter() {defaultValue=100, group=Scifi_Screen_Graph, parameterName=GradientEM_Point01}
   )
  )
  n208 = Multiply(A=NamedRerouteUsage(), B=Add(A=@n200, B=Multiply(A=@n200, B=@n207)))
  n209 = NamedRerouteUsage()
  n210 = Add(A=NamedRerouteDeclaration(None=Multiply(A=@n208, B=@n209)) {name=Graph_Color01, nodeColor=<rgba 0.5234 0 1 1>}, B=NamedRerouteUsage())
  n211 = Add(A=NamedRerouteUsage(), B=NamedRerouteUsage())
  NamedRerouteDeclaration(None=Add(A=Add(A=@n210, B=@n211), B=NamedRerouteUsage())) {name=Graph_Color_All, nodeColor=<rgba 0 1 0.2031 1>}
  n212 = VectorParameter() {
   group=Scifi_Screen_Graph
   parameterName=GraphTexture_Center02
   channelNames {a=ScaleY, b=ScaleX, g=CenterY, r=CenterX}
   defaultValue=<rgba 0.5 0.5 1 1>
  }
  NamedRerouteDeclaration(None=@n212.ScaleY) {name=ScaleY_6, nodeColor=<rgba 0.2422 0 1 1>}
  n213 = ScalarParameter() {defaultValue=20, parameterName=HorizontalDivisor}
  n214 = Multiply(A=NamedRerouteUsage(), B=@n213)
  n215 = Divide(A=Fmod(A=Floor(None=@n214), B=@n213), B=@n213)
  NamedRerouteDeclaration(None=LinearInterpolate(A=NamedRerouteUsage(), B=NamedRerouteUsage(), Alpha=@n215)) {name=XMax_Line, nodeColor=<rgba 0 0.6719 1 1>}
  n216 = NamedRerouteUsage()
  MaterialExpressionAppendVector_16 = AppendVector(A=@n212.CenterX, B=@n212.CenterX)
  MaterialExpressionAppendVector_17 = AppendVector(A=@n212.ScaleX, B=@n212.ScaleX)
  MaterialExpressionNamedRerouteUsage_95 = NamedRerouteUsage()
  n217 = Custom(UVPoint=@MaterialExpressionAppendVector_16, UVScale=@MaterialExpressionAppendVector_17, OriginalUV=@MaterialExpressionNamedRerouteUsage_95) {
   code="float2 DistUV =OriginalUV-UVPoint;\r\nDistUV = DistUV/UVScale + float2(0.5,0.5);\r\n\r\nreturn DistUV;"
   description=CopyToPointUV
   outputType=CMOT_Float2
   inputs [
    {
     inputName=UVPoint
     input {expression=&self:MaterialExpressionAppendVector_16}
    }
    {
     inputName=UVScale
     input {expression=&self:MaterialExpressionAppendVector_17}
    }
    {
     inputName=OriginalUV
     input {expression=&self:MaterialExpressionNamedRerouteUsage_95}
    }
   ]
  }
  n218 = VectorParameter() {group=Scifi_Screen_Graph, parameterName=GraphSpeed02, defaultValue=<rgba 0.1 0 0 0>}
  n219 = TextureObjectParameter() {
   group=Scifi_Screen_Graph
   parameterName=UpperGraph
   samplerType=SAMPLERTYPE_Masks
   texture=&T_HeartLine_Flip
  }
  n220 = TextureSample(UVs=Panner(Coordinate=@n217, Speed=AppendVector(A=@n218.R, B=@n218.R)), Tex=@n219) {
   samplerSource=SSM_Wrap_WorldGroupSettings
   samplerType=SAMPLERTYPE_Masks
   texture=&T_HeartLine
  }
  n221 = Multiply(A=@n220.R, B=ScalarParameter() {defaultValue=0.8, group=Scifi_Screen_Graph, parameterName=BasicEM02})
  n222 = ComponentMask(None=@n217) {r=true}
  n223 = VectorParameter() {group=Scifi_Screen_Graph, parameterName=GraphBoundingBox02, channelNames {a=YMax, b=YMin, g=XMax, r=XMin}, defaultValue=<rgba 0 1 0 1>}
  n224 = Multiply(A=Time(), B=ScalarParameter() {defaultValue=0.5, group=Scifi_Screen_Graph, parameterName=GraphTimeSpeed02})
  n225 = LinearInterpolate(
   A=@n223.XMin
   B=@n223.XMin
   Alpha=StaticSwitchParameter(True=ScalarParameter() {group=Scifi_Screen_Graph, parameterName=00_GraphDebugProgress, sliderMax=1}, False=Frac(None=@n224)) {group=Scifi_Screen_Graph, parameterName=DebugGraph}
  )
  n226 = Multiply(
   A=SphereMask(A=@n222, B=@n225, Radius=ScalarParameter() {defaultValue=0.3, group=Scifi_Screen_Graph, parameterName=GradientRadius02}) {attenuationRadius=0.5, hardnessPercent=0.2}
   B=ScalarParameter() {defaultValue=10, group=Scifi_Screen_Graph, parameterName=GradientEM02}
  )
  n227 = Multiply(
   A=SphereMask(A=@n222, B=@n225, Radius=ScalarParameter() {defaultValue=0.1, group=Scifi_Screen_Graph, parameterName=GradientRadius_Point02}) {attenuationRadius=0.5, hardnessPercent=10}
   B=ScalarParameter() {defaultValue=100, group=Scifi_Screen_Graph, parameterName=GradientEM_Point02}
  )
  n228 = Add(A=@n221, B=Multiply(A=@n221, B=Add(A=@n226, B=@n227)))
  NamedRerouteDeclaration(None=Multiply(A=Multiply(A=@n216, B=@n228), B=NamedRerouteUsage())) {name=Graph_Color02, nodeColor=<rgba 1 0 0.9375 1>}
  n229 = VectorParameter() {
   group=Scifi_Screen_Graph
   parameterName=GraphTexture_Center03
   channelNames {a=ScaleY, b=ScaleX, g=CenterY, r=CenterX}
   defaultValue=<rgba 0.5 0.5 1 1>
  }
  NamedRerouteDeclaration(None=@n229.ScaleY) {name=ScaleY_7, nodeColor=<rgba 0.2422 0 1 1>}
  n230 = NamedRerouteUsage()
  MaterialExpressionAppendVector_21 = AppendVector(A=@n229.CenterX, B=@n229.CenterX)
  MaterialExpressionAppendVector_22 = AppendVector(A=@n229.ScaleX, B=@n229.ScaleX)
  MaterialExpressionNamedRerouteUsage_99 = NamedRerouteUsage()
  n231 = Custom(UVPoint=@MaterialExpressionAppendVector_21, UVScale=@MaterialExpressionAppendVector_22, OriginalUV=@MaterialExpressionNamedRerouteUsage_99) {
   code="float2 DistUV =OriginalUV-UVPoint;\r\nDistUV = DistUV/UVScale + float2(0.5,0.5);\r\n\r\nreturn DistUV;"
   description=CopyToPointUV
   outputType=CMOT_Float2
   inputs [
    {
     inputName=UVPoint
     input {expression=&self:MaterialExpressionAppendVector_21}
    }
    {
     inputName=UVScale
     input {expression=&self:MaterialExpressionAppendVector_22}
    }
    {
     inputName=OriginalUV
     input {expression=&self:MaterialExpressionNamedRerouteUsage_99}
    }
   ]
  }
  n232 = VectorParameter() {group=Scifi_Screen_Graph, parameterName=GraphSpeed03, defaultValue=<rgba 0.1 0 0 0>}
  n233 = TextureObjectParameter() {
   group=Scifi_Screen_Graph
   parameterName=DownGraph
   samplerType=SAMPLERTYPE_Masks
   texture=&T_HeartLine_Flip
  }
  n234 = TextureSample(UVs=Panner(Coordinate=@n231, Speed=AppendVector(A=@n232.R, B=@n232.R)), Tex=@n233) {
   samplerSource=SSM_Wrap_WorldGroupSettings
   samplerType=SAMPLERTYPE_Masks
   texture=&T_HeartLine
  }
  n235 = Multiply(A=@n234.R, B=ScalarParameter() {defaultValue=0.8, group=Scifi_Screen_Graph, parameterName=BasicEM03})
  n236 = ComponentMask(None=@n231) {r=true}
  n237 = VectorParameter() {group=Scifi_Screen_Graph, parameterName=GraphBoundingBox03, channelNames {a=YMax, b=YMin, g=XMax, r=XMin}, defaultValue=<rgba 0 1 0 1>}
  n238 = Multiply(A=Time(), B=ScalarParameter() {defaultValue=0.5, group=Scifi_Screen_Graph, parameterName=GraphTimeSpeed03})
  n239 = LinearInterpolate(
   A=@n237.XMin
   B=@n237.XMin
   Alpha=StaticSwitchParameter(True=ScalarParameter() {group=Scifi_Screen_Graph, parameterName=00_GraphDebugProgress, sliderMax=1}, False=Frac(None=@n238)) {group=Scifi_Screen_Graph, parameterName=DebugGraph}
  )
  n240 = Multiply(
   A=SphereMask(A=@n236, B=@n239, Radius=ScalarParameter() {defaultValue=0.3, group=Scifi_Screen_Graph, parameterName=GradientRadius03}) {attenuationRadius=0.5, hardnessPercent=0.2}
   B=ScalarParameter() {defaultValue=10, group=Scifi_Screen_Graph, parameterName=GradientEM03}
  )
  n241 = Multiply(
   A=SphereMask(A=@n236, B=@n239, Radius=ScalarParameter() {defaultValue=0.1, group=Scifi_Screen_Graph, parameterName=GradientRadius_Point03}) {attenuationRadius=0.5, hardnessPercent=10}
   B=ScalarParameter() {defaultValue=100, group=Scifi_Screen_Graph, parameterName=GradientEM_Point03}
  )
  n242 = Add(A=@n235, B=Multiply(A=@n235, B=Add(A=@n240, B=@n241)))
  NamedRerouteDeclaration(None=Multiply(A=Multiply(A=@n230, B=@n242), B=NamedRerouteUsage())) {name=Graph_Color03, nodeColor=<rgba 1 0 0.9375 1>}
  n243 = VectorParameter() {
   group=Scifi_Screen_Graph
   parameterName=GraphTexture_Center04
   channelNames {a=ScaleY, b=ScaleX, g=CenterY, r=CenterX}
   defaultValue=<rgba 0.5 0.5 1 1>
  }
  NamedRerouteDeclaration(None=@n243.ScaleY) {name=ScaleY_8, nodeColor=<rgba 0.2422 0 1 1>}
  n244 = NamedRerouteUsage()
  MaterialExpressionAppendVector_24 = AppendVector(A=@n243.CenterX, B=@n243.CenterX)
  MaterialExpressionAppendVector_25 = AppendVector(A=@n243.ScaleX, B=@n243.ScaleX)
  MaterialExpressionNamedRerouteUsage_106 = NamedRerouteUsage()
  n245 = Custom(UVPoint=@MaterialExpressionAppendVector_24, UVScale=@MaterialExpressionAppendVector_25, OriginalUV=@MaterialExpressionNamedRerouteUsage_106) {
   code="float2 DistUV =OriginalUV-UVPoint;\r\nDistUV = DistUV/UVScale + float2(0.5,0.5);\r\n\r\nreturn DistUV;"
   description=CopyToPointUV
   outputType=CMOT_Float2
   inputs [
    {
     inputName=UVPoint
     input {expression=&self:MaterialExpressionAppendVector_24}
    }
    {
     inputName=UVScale
     input {expression=&self:MaterialExpressionAppendVector_25}
    }
    {
     inputName=OriginalUV
     input {expression=&self:MaterialExpressionNamedRerouteUsage_106}
    }
   ]
  }
  n246 = VectorParameter() {group=Scifi_Screen_Graph, parameterName=GraphSpeed04, defaultValue=<rgba 0.1 0 0 0>}
  n247 = TextureObjectParameter() {
   group=Scifi_Screen_Graph
   parameterName=DownGraph
   samplerType=SAMPLERTYPE_Masks
   texture=&T_HeartLine_Flip
  }
  n248 = TextureSample(UVs=Panner(Coordinate=@n245, Speed=AppendVector(A=@n246.R, B=@n246.R)), Tex=@n247) {
   samplerSource=SSM_Wrap_WorldGroupSettings
   samplerType=SAMPLERTYPE_Masks
   texture=&T_HeartLine
  }
  n249 = Multiply(A=@n248.R, B=ScalarParameter() {defaultValue=0.8, group=Scifi_Screen_Graph, parameterName=BasicEM04})
  n250 = ComponentMask(None=@n245) {r=true}
  n251 = VectorParameter() {group=Scifi_Screen_Graph, parameterName=GraphBoundingBox04, channelNames {a=YMax, b=YMin, g=XMax, r=XMin}, defaultValue=<rgba 0 1 0 1>}
  n252 = Multiply(A=Time(), B=ScalarParameter() {defaultValue=0.5, group=Scifi_Screen_Graph, parameterName=GraphTimeSpeed04})
  n253 = LinearInterpolate(
   A=@n251.XMax
   B=@n251.XMax
   Alpha=StaticSwitchParameter(True=ScalarParameter() {group=Scifi_Screen_Graph, parameterName=00_GraphDebugProgress, sliderMax=1}, False=Frac(None=@n252)) {group=Scifi_Screen_Graph, parameterName=DebugGraph}
  )
  n254 = Multiply(
   A=SphereMask(A=@n250, B=@n253, Radius=ScalarParameter() {defaultValue=0.3, group=Scifi_Screen_Graph, parameterName=GradientRadius04}) {attenuationRadius=0.5, hardnessPercent=0.2}
   B=ScalarParameter() {defaultValue=10, group=Scifi_Screen_Graph, parameterName=GradientEM04}
  )
  n255 = Multiply(
   A=SphereMask(A=@n250, B=@n253, Radius=ScalarParameter() {defaultValue=0.1, group=Scifi_Screen_Graph, parameterName=GradientRadius_Point04}) {attenuationRadius=0.5, hardnessPercent=10}
   B=ScalarParameter() {defaultValue=100, group=Scifi_Screen_Graph, parameterName=GradientEM_Point04}
  )
  n256 = Add(A=@n249, B=Multiply(A=@n249, B=Add(A=@n254, B=@n255)))
  NamedRerouteDeclaration(None=Multiply(A=Multiply(A=@n244, B=@n256), B=NamedRerouteUsage())) {name=Graph_Color04, nodeColor=<rgba 1 0 0.9375 1>}
  n257 = Constant()
  n258 = If(A=@n97, B=NamedRerouteUsage(), "A > B"=@n257, "A == B"=@n257, "A < B"=Constant() {r=1})
  NamedRerouteDeclaration(None=Multiply(A=Multiply(A=@n100, B=@n258), B=@n106)) {name=AreaMaskStatic, nodeColor=<rgba 0.7578 1 0 1>}
  NamedRerouteUsage()
  OneMinus(None=@n215)
  OneMinus(None=@n48)
  OneMinus(None=@n37)
  OneMinus(None=@n90)
  OneMinus(None=@n79)
 ]
}
