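# Post-process style Animation Blueprint for the M_MED_HydroBottle player body cosmetic, animating its attached boxes and wires.
# It copies the parent mesh pose into a Control Rig weighted by rifle and pistol alphas.
# Additive bone offsets surround a chain of spring-damped, angle-limited AnimDynamics nodes.
# The EventGraph derives weapon and movement alphas, computes gravity override via a deprecated function, and maintains a tagged-pickaxe flag.
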AnimBlueprint /BRCosmetics/Characters/Player/Male/Medium/Bodies/M_MED_HydroBottle/Meshes/M_MED_HydroBottle_AnimBP
target_skeleton=&/M_MED_HydroBottle_Skeleton
graphs [
 AnimGraph {nodes [
   n1 = Control Rig!(
    Source=Copy Pose From Mesh() {bUseAttachedParent=true}
    Alpha
    Is Rifle Alpha:double=Get Is Rifle Alpha() -> (Is Rifle Alpha:float)
    Is Pistol Alpha:double=Get Is Pistol Alpha() -> (Is Pistol Alpha:float)
   ) -> (Pose) {
    alphaBoolBlend {blendOption=Linear}
    controlRigAssetReference {blueprintRigClass=&/M_MED_HydroBottle_CtrlRig.M_MED_HydroBottle_CtrlRig_C}
   }
   n2 = "Transform (Modify) Bone - Bone: dyn_waistbox_c_Wire_MSTR"(ComponentPose=Local To Component(LocalPose=@n1), Alpha=1, Translation="0,0,0", Rotation="0,-10,0", Scale="1,1,1") {
    rotationMode=BMM_Additive
    rotationSpace=BCS_BoneSpace
    translationMode=BMM_Additive
    translationSpace=BCS_BoneSpace
    alphaBoolBlend {blendOption=Linear}
    alphaScaleBiasClamp {bInterpResult=true, interpSpeedIncreasing=4}
    boneToModify {boneName=dyn_waistbox_c_Wire_MSTR}
    rotation {yaw=-10}
   }
   n3 = "Transform (Modify) Bone - Bone: dyn_waistbox_c_Wire_MSTR"(ComponentPose=@n2, Alpha=1, Translation="0,0,0", Rotation="0,-35,0", Scale="1,1,1") {
    rotationMode=BMM_Additive
    rotationSpace=BCS_BoneSpace
    translationMode=BMM_Additive
    translationSpace=BCS_BoneSpace
    alphaBoolBlend {blendOption=Linear}
    boneToModify {boneName=dyn_waistbox_c_Wire_MSTR}
    rotation {yaw=-35}
   }
   n4 = "Transform (Modify) Bone - Bone: dyn_waistbox_c_Wire_MSTR"(ComponentPose=@n3, Alpha=1, Translation="0,0,0", Rotation="0,-10,0", Scale="1,1,1") {
    rotationMode=BMM_Additive
    rotationSpace=BCS_BoneSpace
    translationMode=BMM_Additive
    translationSpace=BCS_BoneSpace
    alphaBoolBlend {blendOption=Linear}
    boneToModify {boneName=dyn_waistbox_c_Wire_MSTR}
    rotation {yaw=-10}
   }
   n5 = "Transform (Modify) Bone - Bone: dyn_waistbox_c_Wire_02"(ComponentPose=@n4, Alpha=1, Translation="0,0,0", Rotation="0,-10,0", Scale="1,1,1") {
    rotationMode=BMM_Additive
    rotationSpace=BCS_BoneSpace
    translationMode=BMM_Additive
    translationSpace=BCS_BoneSpace
    alphaBoolBlend {blendOption=Linear}
    boneToModify {boneName=dyn_waistbox_c_Wire_02}
    rotation {yaw=-10}
   }
   n6 = Anim Dynamics - Bone: dyn_claviclebox_l!(ComponentPose=@n5, Alpha=1, GravityOverride:Vector="0,0,0", ExternalForce="0,0,0") -> (Pose) {
    angularDampingOverride=0.9999
    angularSpringConstant=200
    bAngularSpring=true
    bOverrideAngularDamping=true
    bUseGravityOverride=true
    simulationSpace=RootRelative
    alphaBoolBlend {blendOption=Linear}
    boundBone {boneName=dyn_claviclebox_l}
    componentAppliedLinearAccClamp=<6000 6000 6000>
    componentLinearAccScale=<0.3 0.3 0.4>
    componentLinearVelScale=<0.3 0.3 0.4>
    physicsBodyDefinitions [{
      boundBone {boneName=dyn_claviclebox_l}
      boxExtents=<10 10 10>
      constraintSetup {angularLimitsMax=<5 5 2.5>, angularLimitsMin=<-5 -5 -2.5>, angularTarget=<1 0 0>}
      localJointOffset=<4 0 0>
     }]
   }
   n7 = Anim Dynamics - Bone: dyn_claviclebox_r!(ComponentPose=@n6, Alpha=1, GravityOverride:Vector="0,0,0", ExternalForce="0,0,0") -> (Pose) {
    angularDampingOverride=0.9999
    angularSpringConstant=200
    bAngularSpring=true
    bOverrideAngularDamping=true
    bUseGravityOverride=true
    simulationSpace=RootRelative
    alphaBoolBlend {blendOption=Linear}
    boundBone {boneName=dyn_claviclebox_r}
    componentAppliedLinearAccClamp=<6000 6000 6000>
    componentLinearAccScale=<0.3 0.3 0.4>
    componentLinearVelScale=<0.3 0.3 0.4>
    physicsBodyDefinitions [{
      boundBone {boneName=dyn_claviclebox_r}
      boxExtents=<10 10 10>
      constraintSetup {angularLimitsMax=<5 5 2.5>, angularLimitsMin=<-5 -5 -2.5>, angularTarget=<1 0 0>}
      localJointOffset=<4 0 0>
     }]
   }
   n8 = Anim Dynamics - Bone: torsobox_r!(ComponentPose=@n7, Alpha=1, GravityOverride:Vector="0,0,0", ExternalForce="0,0,0") -> (Pose) {
    angularDampingOverride=0.9999
    angularSpringConstant=200
    bAngularSpring=true
    bOverrideAngularDamping=true
    bUseGravityOverride=true
    simulationSpace=RootRelative
    alphaBoolBlend {blendOption=Linear}
    boundBone {boneName=torsobox_r}
    componentAppliedLinearAccClamp=<6000 6000 6000>
    componentLinearAccScale=<0.3 0.3 0.4>
    componentLinearVelScale=<0.3 0.3 0.4>
    physicsBodyDefinitions [{
      boundBone {boneName=torsobox_r}
      boxExtents=<10 10 10>
      constraintSetup {angularLimitsMax=<5 5 5>, angularLimitsMin=<-5 -5 -5>, angularTarget=<1 0 0>}
      localJointOffset=<4 0 0>
     }]
   }
   n9 = Anim Dynamics - Bone: waistbox_l!(ComponentPose=@n8, Alpha=1, GravityOverride:Vector="0,0,0", ExternalForce="0,0,0") -> (Pose) {
    angularDampingOverride=0.9999
    angularSpringConstant=200
    bAngularSpring=true
    bOverrideAngularDamping=true
    bUseGravityOverride=true
    simulationSpace=RootRelative
    alphaBoolBlend {blendOption=Linear}
    boundBone {boneName=waistbox_l}
    componentAppliedLinearAccClamp=<6000 6000 6000>
    componentLinearAccScale=<0.3 0.3 0.4>
    componentLinearVelScale=<0.3 0.3 0.4>
    physicsBodyDefinitions [{
      boundBone {boneName=waistbox_l}
      boxExtents=<10 10 10>
      constraintSetup {angularLimitsMax=<5 5 2.5>, angularLimitsMin=<-5 -5 -2.5>, angularTarget=<1 0 0>}
      localJointOffset=<4 0 0>
     }]
   }
   n10 = Anim Dynamics - Bone: waistbox_r!(ComponentPose=@n9, Alpha=1, GravityOverride:Vector="0,0,0", ExternalForce="0,0,0") -> (Pose) {
    angularDampingOverride=0.9999
    angularSpringConstant=200
    bAngularSpring=true
    bOverrideAngularDamping=true
    bUseGravityOverride=true
    simulationSpace=RootRelative
    alphaBoolBlend {blendOption=Linear}
    boundBone {boneName=waistbox_r}
    componentAppliedLinearAccClamp=<6000 6000 6000>
    componentLinearAccScale=<0.3 0.3 0.4>
    componentLinearVelScale=<0.3 0.3 0.4>
    physicsBodyDefinitions [{
      boundBone {boneName=waistbox_r}
      boxExtents=<10 10 10>
      constraintSetup {angularLimitsMax=<5 5 5>, angularLimitsMin=<-5 -5 -5>, angularTarget=<1 0 0>}
      localJointOffset=<4 0 0>
     }]
   }
   n11 = Anim Dynamics - Bone: waistbox_c!(ComponentPose=@n10, Alpha=1, GravityOverride:Vector="0,0,0", ExternalForce="0,0,0") -> (Pose) {
    angularDampingOverride=0.9999
    angularSpringConstant=200
    bAngularSpring=true
    bOverrideAngularDamping=true
    bUseGravityOverride=true
    simulationSpace=RootRelative
    alphaBoolBlend {blendOption=Linear}
    boundBone {boneName=waistbox_c}
    componentAppliedLinearAccClamp=<6000 6000 6000>
    componentLinearAccScale=<0.3 0.3 0.4>
    componentLinearVelScale=<0.3 0.3 0.4>
    physicsBodyDefinitions [{
      boundBone {boneName=waistbox_c}
      boxExtents=<10 10 10>
      constraintSetup {angularLimitsMax=<5 5 2.5>, angularLimitsMin=<-5 -5 -2.5>, angularTarget=<1 0 0>}
      localJointOffset=<4 0 0>
     }]
   }
   n12 = Anim Dynamics - Bone: legbox_01_l!(ComponentPose=@n11, Alpha=0.25, GravityOverride:Vector="0,0,0", ExternalForce="0,0,0") -> (Pose) {
    angularDampingOverride=1
    angularSpringConstant=50
    bAngularSpring=true
    bOverrideAngularDamping=true
    bUseGravityOverride=true
    simulationSpace=RootRelative
    alphaBoolBlend {blendOption=Linear}
    boundBone {boneName=legbox_01_l}
    componentAppliedLinearAccClamp=<6000 6000 6000>
    componentLinearAccScale=<0.3 0.3 0.4>
    componentLinearVelScale=<0.3 0.3 0.4>
    physicsBodyDefinitions [{
      boundBone {boneName=legbox_01_l}
      boxExtents=<10 10 10>
      constraintSetup {angularLimitsMax=<10 5 25>, angularLimitsMin=<-10 -25 -25>, angularTarget=<1 0 0>}
      localJointOffset=<4 0 0>
     }]
    relativeSpaceBone {boneName=pelvis}
   }
   n13 = Anim Dynamics - Bone: legbox_02_l!(ComponentPose=@n12, Alpha=0.25, GravityOverride:Vector="0,0,0", ExternalForce="0,0,0") -> (Pose) {
    angularDampingOverride=1
    angularSpringConstant=200
    bAngularSpring=true
    bOverrideAngularDamping=true
    bUseGravityOverride=true
    simulationSpace=RootRelative
    alphaBoolBlend {blendOption=Linear}
    boundBone {boneName=legbox_02_l}
    componentAppliedLinearAccClamp=<6000 6000 6000>
    componentLinearAccScale=<0.3 0.3 0.4>
    componentLinearVelScale=<0.3 0.3 0.4>
    physicsBodyDefinitions [{
      boundBone {boneName=legbox_02_l}
      boxExtents=<10 10 10>
      constraintSetup {angularLimitsMax=<10 5 25>, angularLimitsMin=<-10 -25 -25>, angularTarget=<1 0 0>}
      localJointOffset=<4 0 0>
     }]
    relativeSpaceBone {boneName=pelvis}
   }
   n14 = Anim Dynamics - Bone: legbox_r!(ComponentPose=@n13, Alpha=0.5, GravityOverride:Vector="0,0,0", ExternalForce="0,0,0") -> (Pose) {
    angularDampingOverride=0.9999
    angularSpringConstant=200
    bAngularSpring=true
    bOverrideAngularDamping=true
    bUseGravityOverride=true
    simulationSpace=RootRelative
    alphaBoolBlend {blendOption=Linear}
    boundBone {boneName=legbox_r}
    componentAppliedLinearAccClamp=<6000 6000 6000>
    componentLinearAccScale=<0.3 0.3 0.4>
    componentLinearVelScale=<0.3 0.3 0.4>
    physicsBodyDefinitions [{
      boundBone {boneName=legbox_r}
      boxExtents=<10 10 10>
      constraintSetup {angularLimitsMax=<5 5 5>, angularLimitsMin=<-5 -5 -5>, angularTarget=<1 0 0>}
      localJointOffset=<4 0 0>
     }]
   }
   n15 = "Transform (Modify) Bone - Bone: dyn_shoulderwire_l_MSTR"(ComponentPose=@n14, Alpha=1, Translation="0,0,0", Rotation="-6.5,-25,0", Scale="1,1,1") {
    rotationMode=BMM_Additive
    rotationSpace=BCS_BoneSpace
    translationMode=BMM_Additive
    translationSpace=BCS_BoneSpace
    alphaBoolBlend {blendOption=Linear}
    boneToModify {boneName=dyn_shoulderwire_l_MSTR}
    rotation {pitch=-6.5, yaw=-25}
   }
   n16 = Anim Dynamics - Bone: dyn_shoulderwire_01_l!(ComponentPose=@n15, Alpha=0.5, GravityOverride:Vector="0,0,0", ExternalForce="0,0,0") -> (Pose) {
    angularDampingOverride=1
    angularSpringConstant=50
    bAngularSpring=true
    bOverrideAngularDamping=true
    bUseGravityOverride=true
    simulationSpace=RootRelative
    alphaBoolBlend {blendOption=Linear}
    boundBone {boneName=dyn_shoulderwire_01_l}
    componentAppliedLinearAccClamp=<6000 6000 6000>
    componentLinearAccScale=<0.3 0.3 0.4>
    componentLinearVelScale=<0.3 0.3 0.4>
    physicsBodyDefinitions [{
      boundBone {boneName=dyn_shoulderwire_01_l}
      boxExtents=<10 10 10>
      constraintSetup {angularLimitsMax=<1 5 5>, angularLimitsMin=<-1 -15 -75>, angularTarget=<1 0 0>}
      localJointOffset=<2 0 0>
     }]
   }
   n17 = Anim Dynamics - Bone: dyn_shoulderwire_02_l!(ComponentPose=@n16, Alpha=0.5, GravityOverride:Vector="0,0,0", ExternalForce="0,0,0") -> (Pose) {
    angularDampingOverride=1
    angularSpringConstant=50
    bAngularSpring=true
    bOverrideAngularDamping=true
    bUseGravityOverride=true
    simulationSpace=RootRelative
    alphaBoolBlend {blendOption=Linear}
    boundBone {boneName=dyn_shoulderwire_02_l}
    componentAppliedLinearAccClamp=<6000 6000 6000>
    componentLinearAccScale=<0.3 0.3 0.4>
    componentLinearVelScale=<0.3 0.3 0.4>
    physicsBodyDefinitions [{
      boundBone {boneName=dyn_shoulderwire_02_l}
      boxExtents=<10 10 10>
      constraintSetup {angularLimitsMax=<1 5 5>, angularLimitsMin=<-1 -15 -75>, angularTarget=<1 0 0>}
      localJointOffset=<2 0 0>
     }]
   }
   n18 = Anim Dynamics - Bone: dyn_shoulderwire_03_l!(ComponentPose=@n17, Alpha=0.5, GravityOverride:Vector="0,0,0", ExternalForce="0,0,0") -> (Pose) {
    angularDampingOverride=1
    angularSpringConstant=50
    bAngularSpring=true
    bOverrideAngularDamping=true
    bUseGravityOverride=true
    simulationSpace=RootRelative
    alphaBoolBlend {blendOption=Linear}
    boundBone {boneName=dyn_shoulderwire_03_l}
    componentAppliedLinearAccClamp=<6000 6000 6000>
    componentLinearAccScale=<0.3 0.3 0.4>
    componentLinearVelScale=<0.3 0.3 0.4>
    physicsBodyDefinitions [{
      boundBone {boneName=dyn_shoulderwire_03_l}
      boxExtents=<10 10 10>
      constraintSetup {angularLimitsMax=<1 5 5>, angularLimitsMin=<-1 -15 -75>, angularTarget=<1 0 0>}
      localJointOffset=<2 0 0>
     }]
   }
   n19 = Anim Dynamics - Bone: dyn_legbox_l_wire_01!(ComponentPose=@n18, Alpha=0.5, GravityOverride:Vector="0,0,0", ExternalForce="0,0,0") -> (Pose) {
    angularDampingOverride=0.9999
    angularSpringConstant=200
    bAngularSpring=true
    bOverrideAngularDamping=true
    bUseGravityOverride=true
    simulationSpace=RootRelative
    alphaBoolBlend {blendOption=Linear}
    boundBone {boneName=dyn_legbox_l_wire_01}
    componentAppliedLinearAccClamp=<6000 6000 6000>
    componentLinearAccScale=<0.3 0.3 0.4>
    componentLinearVelScale=<0.3 0.3 0.4>
    physicsBodyDefinitions [{
      boundBone {boneName=dyn_legbox_l_wire_01}
      boxExtents=<10 10 10>
      constraintSetup {angularLimitsMax=<5 5 5>, angularLimitsMin=<-5 -25 -5>, angularTarget=<1 0 0>}
      localJointOffset=<4 0 0>
     }]
   }
   n20 = Anim Dynamics - Bone: dyn_legbox_l_wire_02!(ComponentPose=@n19, Alpha=0.5, GravityOverride:Vector="0,0,0", ExternalForce="0,0,0") -> (Pose) {
    angularDampingOverride=0.9999
    angularSpringConstant=200
    bAngularSpring=true
    bOverrideAngularDamping=true
    bUseGravityOverride=true
    simulationSpace=RootRelative
    alphaBoolBlend {blendOption=Linear}
    boundBone {boneName=dyn_legbox_l_wire_02}
    componentAppliedLinearAccClamp=<6000 6000 6000>
    componentLinearAccScale=<0.3 0.3 0.4>
    componentLinearVelScale=<0.3 0.3 0.4>
    physicsBodyDefinitions [{
      boundBone {boneName=dyn_legbox_l_wire_02}
      boxExtents=<10 10 10>
      constraintSetup {angularLimitsMax=<5 5 5>, angularLimitsMin=<-5 -25 -5>, angularTarget=<1 0 0>}
      localJointOffset=<4 0 0>
     }]
   }
   n21 = Anim Dynamics - Bone: dyn_legbox_l_wire_03!(ComponentPose=@n20, Alpha=0.5, GravityOverride:Vector="0,0,0", ExternalForce="0,0,0") -> (Pose) {
    angularDampingOverride=0.9999
    angularSpringConstant=200
    bAngularSpring=true
    bOverrideAngularDamping=true
    bUseGravityOverride=true
    simulationSpace=RootRelative
    alphaBoolBlend {blendOption=Linear}
    boundBone {boneName=dyn_legbox_l_wire_03}
    componentAppliedLinearAccClamp=<6000 6000 6000>
    componentLinearAccScale=<0.3 0.3 0.4>
    componentLinearVelScale=<0.3 0.3 0.4>
    physicsBodyDefinitions [{
      boundBone {boneName=dyn_legbox_l_wire_03}
      boxExtents=<10 10 10>
      constraintSetup {angularLimitsMax=<5 5 5>, angularLimitsMin=<-5 -25 -5>, angularTarget=<1 0 0>}
      localJointOffset=<4 0 0>
     }]
   }
   n22 = Anim Dynamics - Bone: dyn_waistbox_c_Wire_01!(ComponentPose=@n21, Alpha=0.5, GravityOverride:Vector="0,0,0", ExternalForce="0,0,0") -> (Pose) {
    angularDampingOverride=0.9999
    angularSpringConstant=200
    bAngularSpring=true
    bOverrideAngularDamping=true
    bUseGravityOverride=true
    simulationSpace=RootRelative
    alphaBoolBlend {blendOption=Linear}
    boundBone {boneName=dyn_waistbox_c_Wire_01}
    componentAppliedLinearAccClamp=<6000 6000 6000>
    componentLinearAccScale=<0.3 0.3 0.4>
    componentLinearVelScale=<0.3 0.3 0.4>
    physicsBodyDefinitions [{
      boundBone {boneName=dyn_waistbox_c_Wire_01}
      boxExtents=<10 10 10>
      constraintSetup {angularLimitsMax=<2.5 2.5 5>, angularLimitsMin=<-2.5 -2.5 -50>, angularTarget=<1 0 0>}
      localJointOffset=<8 0 0>
     }]
   }
   n23 = Anim Dynamics - Bone: dyn_waistbox_c_Wire_02!(ComponentPose=@n22, Alpha=0.5, GravityOverride:Vector="0,0,0", ExternalForce="0,0,0") -> (Pose) {
    angularDampingOverride=0.9999
    angularSpringConstant=200
    bAngularSpring=true
    bOverrideAngularDamping=true
    bUseGravityOverride=true
    simulationSpace=RootRelative
    alphaBoolBlend {blendOption=Linear}
    boundBone {boneName=dyn_waistbox_c_Wire_02}
    componentAppliedLinearAccClamp=<6000 6000 6000>
    componentLinearAccScale=<0.3 0.3 0.4>
    componentLinearVelScale=<0.3 0.3 0.4>
    physicsBodyDefinitions [{
      boundBone {boneName=dyn_waistbox_c_Wire_02}
      boxExtents=<10 10 10>
      constraintSetup {angularLimitsMax=<2.5 2.5 5>, angularLimitsMin=<-2.5 -2.5 -50>, angularTarget=<1 0 0>}
      localJointOffset=<8 0 0>
     }]
   }
   n24 = Anim Dynamics - Bone: dyn_waistbox_c_Wire_03!(ComponentPose=@n23, Alpha=0.5, GravityOverride:Vector="0,0,0", ExternalForce="0,0,0") -> (Pose) {
    angularDampingOverride=0.9999
    angularSpringConstant=200
    bAngularSpring=true
    bOverrideAngularDamping=true
    bUseGravityOverride=true
    simulationSpace=RootRelative
    alphaBoolBlend {blendOption=Linear}
    boundBone {boneName=dyn_waistbox_c_Wire_03}
    componentAppliedLinearAccClamp=<6000 6000 6000>
    componentLinearAccScale=<0.3 0.3 0.4>
    componentLinearVelScale=<0.3 0.3 0.4>
    physicsBodyDefinitions [{
      boundBone {boneName=dyn_waistbox_c_Wire_03}
      boxExtents=<10 10 10>
      constraintSetup {angularLimitsMax=<2.5 2.5 25>, angularLimitsMin=<-2.5 -2.5 -25>, angularTarget=<1 0 0>}
      localJointOffset=<4 0 0>
     }]
   }
   n25 = Anim Dynamics - Bone: dyn_waistbox_c_Wire_04!(ComponentPose=@n24, Alpha=0.5, GravityOverride:Vector="0,0,0", ExternalForce="0,0,0") -> (Pose) {
    angularDampingOverride=0.9999
    angularSpringConstant=200
    bAngularSpring=true
    bOverrideAngularDamping=true
    bUseGravityOverride=true
    simulationSpace=RootRelative
    alphaBoolBlend {blendOption=Linear}
    boundBone {boneName=dyn_waistbox_c_Wire_04}
    componentAppliedLinearAccClamp=<6000 6000 6000>
    componentLinearAccScale=<0.3 0.3 0.4>
    componentLinearVelScale=<0.3 0.3 0.4>
    physicsBodyDefinitions [{
      boundBone {boneName=dyn_waistbox_c_Wire_04}
      boxExtents=<10 10 10>
      constraintSetup {angularLimitsMax=<2.5 2.5 25>, angularLimitsMin=<-2.5 -2.5 -25>, angularTarget=<1 0 0>}
      localJointOffset=<4 0 0>
     }]
   }
   n26 = Anim Dynamics - Bone: dyn_claviclebox_wire_r_01!(ComponentPose=@n25, Alpha=0.25, GravityOverride:Vector="0,0,0", ExternalForce="0,0,0") -> (Pose) {
    angularDampingOverride=0.9999
    angularSpringConstant=200
    bAngularSpring=true
    bOverrideAngularDamping=true
    bUseGravityOverride=true
    simulationSpace=RootRelative
    alphaBoolBlend {blendOption=Linear}
    boundBone {boneName=dyn_claviclebox_wire_r_01}
    componentAppliedLinearAccClamp=<6000 6000 6000>
    componentLinearAccScale=<0.3 0.3 0.4>
    componentLinearVelScale=<0.3 0.3 0.4>
    physicsBodyDefinitions [{
      boundBone {boneName=dyn_claviclebox_wire_r_01}
      boxExtents=<10 10 10>
      constraintSetup {angularLimitsMax=<2.5 2.5 5>, angularLimitsMin=<-2.5 -2.5 -25>, angularTarget=<1 0 0>}
      localJointOffset=<4 0 0>
     }]
   }
   n27 = Anim Dynamics - Bone: dyn_claviclebox_wire_r_02!(ComponentPose=@n26, Alpha=0.25, GravityOverride:Vector="0,0,0", ExternalForce="0,0,0") -> (Pose) {
    angularDampingOverride=0.9999
    angularSpringConstant=200
    bAngularSpring=true
    bOverrideAngularDamping=true
    bUseGravityOverride=true
    simulationSpace=RootRelative
    alphaBoolBlend {blendOption=Linear}
    boundBone {boneName=dyn_claviclebox_wire_r_02}
    componentAppliedLinearAccClamp=<6000 6000 6000>
    componentLinearAccScale=<0.3 0.3 0.4>
    componentLinearVelScale=<0.3 0.3 0.4>
    physicsBodyDefinitions [{
      boundBone {boneName=dyn_claviclebox_wire_r_02}
      boxExtents=<10 10 10>
      constraintSetup {angularLimitsMax=<2.5 2.5 5>, angularLimitsMin=<-2.5 -2.5 -25>, angularTarget=<1 0 0>}
      localJointOffset=<4 0 0>
     }]
   }
   n28 = Anim Dynamics - Bone: dyn_claviclebox_wire_r_03!(ComponentPose=@n27, Alpha=0.25, GravityOverride:Vector="0,0,0", ExternalForce="0,0,0") -> (Pose) {
    angularDampingOverride=0.9999
    angularSpringConstant=200
    bAngularSpring=true
    bOverrideAngularDamping=true
    bUseGravityOverride=true
    simulationSpace=RootRelative
    alphaBoolBlend {blendOption=Linear}
    boundBone {boneName=dyn_claviclebox_wire_r_03}
    componentAppliedLinearAccClamp=<6000 6000 6000>
    componentLinearAccScale=<0.3 0.3 0.4>
    componentLinearVelScale=<0.3 0.3 0.4>
    physicsBodyDefinitions [{
      boundBone {boneName=dyn_claviclebox_wire_r_03}
      boxExtents=<10 10 10>
      constraintSetup {angularLimitsMax=<2.5 2.5 5>, angularLimitsMin=<-2.5 -2.5 -25>, angularTarget=<1 0 0>}
      localJointOffset=<4 0 0>
     }]
   }
   n29 = "Transform (Modify) Bone - Bone: dyn_waistbox_c_Wire_MSTR"(ComponentPose=@n28, Alpha=1, Translation="0,0,0", Rotation="0,-90,0", Scale="1,1,1") {
    rotationMode=BMM_Additive
    rotationSpace=BCS_BoneSpace
    translationMode=BMM_Additive
    translationSpace=BCS_BoneSpace
    alphaBoolBlend {blendOption=Linear}
    boneToModify {boneName=dyn_waistbox_c_Wire_MSTR}
    rotation {yaw=-90}
   }
   n30 = "Transform (Modify) Bone - Bone: legbox_01_l"(ComponentPose=@n29, Alpha=1, Translation="0,-2,0", Rotation="30,22,35", Scale="1,1,1") {
    rotationMode=BMM_Additive
    rotationSpace=BCS_BoneSpace
    translationMode=BMM_Additive
    translationSpace=BCS_BoneSpace
    alphaBoolBlend {blendOption=Linear}
    alphaScaleBiasClamp {bInterpResult=true, interpSpeedIncreasing=5}
    boneToModify {boneName=legbox_01_l}
    rotation=<rot 30 22 35>
    translation {y=-2}
   }
   n31 = "Transform (Modify) Bone - Bone: dyn_legbox_l_wire_01"(ComponentPose=@n30, Alpha=1, Translation="0,0,0", Rotation="50,0,0", Scale="1,1,1") {
    rotationMode=BMM_Additive
    rotationSpace=BCS_BoneSpace
    translationMode=BMM_Additive
    translationSpace=BCS_BoneSpace
    alphaBoolBlend {blendOption=Linear}
    boneToModify {boneName=dyn_legbox_l_wire_01}
    rotation {pitch=50}
   }
   n32 = "Transform (Modify) Bone - Bone: dyn_waistbox_c_Wire_MSTR"(ComponentPose=@n31, Alpha=1, Translation="0,0,0", Rotation="0,-120,30", Scale="1,1,1") {
    rotationMode=BMM_Additive
    rotationSpace=BCS_BoneSpace
    translationMode=BMM_Additive
    translationSpace=BCS_BoneSpace
    alphaBoolBlend {blendOption=Linear}
    boneToModify {boneName=dyn_waistbox_c_Wire_MSTR}
    rotation {roll=30, yaw=-120}
   }
   n33 = "Transform (Modify) Bone - Bone: legbox_01_l"(ComponentPose=@n32, Alpha=1, Translation="0,-2,0", Rotation="30,22,35", Scale="1,1,1") {
    rotationMode=BMM_Additive
    rotationSpace=BCS_BoneSpace
    translationMode=BMM_Additive
    translationSpace=BCS_BoneSpace
    alphaBoolBlend {blendOption=Linear}
    alphaScaleBiasClamp {bInterpResult=true, interpSpeedIncreasing=5}
    boneToModify {boneName=legbox_01_l}
    rotation=<rot 30 22 35>
    translation {y=-2}
   }
   n34 = "Transform (Modify) Bone - Bone: dyn_legbox_l_wire_01"(ComponentPose=@n33, Alpha=1, Translation="0,0,0", Rotation="50,0,-30", Scale="1,1,1") {
    rotationMode=BMM_Additive
    rotationSpace=BCS_BoneSpace
    translationMode=BMM_Additive
    translationSpace=BCS_BoneSpace
    alphaBoolBlend {blendOption=Linear}
    boneToModify {boneName=dyn_legbox_l_wire_01}
    rotation {pitch=50, roll=-30}
   }
   Output Pose(Result=Component To Local(ComponentPose=@n34)) {layerGroup=None}
  ]}
 EventGraph {nodes [
   Event BlueprintUpdateAnimation()
   > n1 = Sequence!(execute) -> (then_0, then_1, then_2:Exec)
   > n2 = Gravity Override State Function DEPRECATED(CharacterPartAnimInstance=Self-Reference(), GameTimeInSeconds=GetGameTimeInSeconds(), GravityOverrideParams=Get Gravity Override Params())
   > Set GravityOverride(GravityOverride=@n2)
   Set Is Knee Sliding Alpha(execute=@n1.then_1, Is Knee Sliding Alpha:double="To Float (Boolean)"(InBool=Get bIsKneeSliding())) -> (Output_Get:double)
   > Set Is Playing Emote Alpha(Is Playing Emote Alpha:double="To Float (Boolean)"(InBool=Get bIsPlayingEmote())) -> (Output_Get:double)
   > Set Is Skydiving Alpha(Is Skydiving Alpha:double="To Float (Boolean)"(InBool=Get bIsSkyDiving())) -> (Output_Get:double)
   > Set Is Baseball Sliding(Is Baseball Sliding:double="To Float (Boolean)"(InBool=Get bIsBaseballSliding())) -> (Output_Get:double)
   n3 = Sequence!(execute=@n1.then_2) -> (then_0, then_1, then_2, then_3:Exec, then_4:Exec)
   > n4 = Cast To FortPlayerAnimInstance(Object=GetAnimInstance(self=GetOwnerSkeletalMeshComponent()))
   > Set isPistol(isPistol=false)
   > Set isRifle(isRifle:bool=false) -> (Output_Get:bool)
   > n5 = Switch on EFortWeaponCoreAnimation!(execute, Selection:EFortWeaponCoreAnimation Enum=Get WeaponCoreAnim(self=@n4)) -> (
    Melee:Exec
    Pistol:Exec
    Shotgun:Exec
    PaperBlueprint:Exec
    Rifle:Exec
    MeleeOneHand:Exec
    MachinePistol:Exec
    RocketLauncher:Exec
    GrenadeLauncher:Exec
    GoingCommando:Exec
    AssaultRifle:Exec
    TacticalShotgun:Exec
    SniperRifle:Exec
    TrapPlacement:Exec
    ShoulderLauncher:Exec
    AbilityDecoTool:Exec
    Crossbow:Exec
    C4:Exec
    RemoteControl:Exec
    DualWield:Exec
    AR_BullPup:Exec
    AR_ForwardGrip:Exec
    MedPackPaddles:Exec
    SMG_P90:Exec
    AR_DrumGun:Exec
    Consumable_Small:Exec
    Consumable_Large:Exec
    Balloon:Exec
    MountedTurret:Exec
    CreativeTool:Exec
    ExplosiveBow:Exec
    AshtonIndigo:Exec
    AshtonChicago:Exec
    MeleeDualWield:Exec
    Unarmed:Exec
   )
   Set isRifle(execute=[@n5.AssaultRifle, @n5.Rifle, @n5.RocketLauncher, @n5.Shotgun, @n5.SniperRifle, @n5.TacticalShotgun], isRifle=true)
   Set isPistol(execute=@n5.Pistol, isPistol=true)
   Set Is Rifle Alpha(execute=@n3.then_1, Is Rifle Alpha:float="To Float (Boolean)"(InBool=Get isRifle() -> (isRifle:bool))) -> (Output_Get:float)
   > Set Is Pistol Alpha(Is Pistol Alpha:float="To Float (Boolean)"(InBool=Get isPistol() -> (isPistol:bool))) -> (Output_Get:float)
   n6 = Branch(execute=@n3.then_2, Condition=Get bIsSkyDiving())
   > Set Is Skydiving Float(Is Skydiving Float:double=0.25) -> (Output_Get:double)
   Set Is Skydiving Float(execute=@n6.else, Is Skydiving Float=0.5)
   n7 = NOT Boolean(A=Get bIsSkyDiving())
   n8 = OR Boolean(A=Get bIsJumping(), B=Get bIsFalling())
   Set Is JumpingFalling Alpha(execute=@n3.then_3, Is JumpingFalling Alpha:double="To Float (Boolean)"(InBool=AND Boolean(A=@n8, B=@n7))) -> (Output_Get:double)
   n9 = Event OnCharacterAndPickaxeCustomizationLoaded!() -> (
    OutputDelegate
    then
    CustomBaseMesh:Skeletal Mesh Component Object Reference
    CustomParts:Array of Custom Character Part Object References
    CustomMeshParts:Array of Skeletal Mesh Component Budgeted Object References
   ) {is=title K2Node_Event}
   Branch(execute=[@n9.then, @n10.then, @n11.then, @n14.then], Condition=Get bIsSkyDiving())
   n10 = Event OnEndPlayingEmote() {is=title K2Node_Event}
   n11 = Event OnEndSkydiving()
   Event OnBeginPlayingEmote() {is=title K2Node_Event}
   > n12 = Delay(Duration=0.2)
   Set HydroBottlePickaxe(execute=[@n12.then, @n13.then], HydroBottlePickaxe:bool=false) -> (Output_Get:bool)
   n13 = Event OnBeginSkydiving()
   n14 = Event OnWeaponChanged()
   Set HydroBottlePickaxe(HydroBottlePickaxe=PickaxeItemDefinitionHasAnyMetaTag(Tags=Make Array("[0]"=(TagName="Cosmetics.MetaTags.PeacefulPoem")), bRequiresEquipped=true))
  ]}
]
variables [
 Gravity Override Params {type=GravityOverrideParamsStruct}
 GravityOverride {type=Vector}
 HydroBottlePickaxe {type=boolean}
 Is Baseball Sliding {type=number}
 Is JumpingFalling Alpha {type=number}
 Is Knee Sliding Alpha {type=number}
 Is Pistol Alpha {type=number}
 Is Playing Emote Alpha {type=number}
 Is Rifle Alpha {type=number}
 Is Skydiving Alpha {type=number}
 Is Skydiving Float {type=number}
 isPistol {type=boolean}
 isRifle {type=boolean}
]
